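# Niagara system for the CosmicStars cosmetic aura, drawing glow sprites on bones or sockets of a skeletal mesh.
# Decal_Glows emitters burst-spawn grouped sprites, using spawn-group switches and partition arrays to set size, offset, bone and tint per layer.
# User flags adjust for large bodies and front-end display, while time-of-day curves modulate glow brightness.
# In the visible portion, only Decal_Glows_Combined002 is enabled; other glow variants and waveform-animated Radial_Beams are disabled.
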NiagaraSystem /SparksCosmetics/Items/Auras/Assets/NS_Aura_CosmicStars/NS_Aura_CosmicStars
dependencies {
 usedDataInterfaces [
  &NiagaraDataInterfaceCurve
  /Script/Niagara.NiagaraDataInterfaceArrayDistributionInt
  &NiagaraDataInterfaceArrayFloat
  &NiagaraDataInterfaceArrayInt32
  &NiagaraDataInterfaceArrayFloat3
  &NiagaraDataInterfaceColorCurve
  &NiagaraDataInterfaceSkeletalMesh
 ]
 usedDynamicInputs [&MakeVector, &MakeCustomVectorFromBool]
 usedModules [
  &self:SystemSpawnScript.SystemScriptSource.SystemScriptGraph.NiagaraNodeAssignment_0.SetVariables_#1
  &SystemState
  &self:SystemSpawnScript.SystemScriptSource.SystemScriptGraph.NiagaraNodeAssignment_4.SetVariables_#2
  &self:SystemSpawnScript.SystemScriptSource.SystemScriptGraph.NiagaraNodeAssignment_1.SetVariables_#3
  &self:SystemSpawnScript.SystemScriptSource.SystemScriptGraph.NiagaraNodeAssignment_5.SetVariables_#4
  &self:SystemSpawnScript.SystemScriptSource.SystemScriptGraph.NiagaraNodeAssignment_2.SetVariables_#5
  &self:SystemSpawnScript.SystemScriptSource.SystemScriptGraph.NiagaraNodeAssignment_6.SetVariables_#6
  &SampleSkeletalMesh
  &self:SystemSpawnScript.SystemScriptSource.SystemScriptGraph.NiagaraNodeAssignment_3.SetVariables_#7
 ]
}
emitters {
 Decal_Glows_Combined {
  bEnabled=false
  emitterUpdateScript [
   &EmitterState()
   &SpawnBurst_Instantaneous(Spawn Count=5, Use Spawn Probability=false, UseLoopCountLimit=true)
   &SpawnBurst_Instantaneous(Spawn Group=1, Use Spawn Probability=false, UseLoopCountLimit=true) {module=SpawnBurst_Instantaneous001}
   &SpawnBurst_Instantaneous(Spawn Group=2, Use Spawn Probability=false, UseLoopCountLimit=true) {module=SpawnBurst_Instantaneous002}
   &SpawnBurst_Instantaneous(Spawn Group=3, Use Spawn Probability=false, UseLoopCountLimit=true) {module=SpawnBurst_Instantaneous003}
  ]
  particleSpawnScript [
   /SetVariables(Particles.VisibilityTag=&SpawnGroup()) {isSetParameters=true, module=SetVariables_#8}
   &PartitionParticles(Number Of Partitions=5)
   &PartitionParticles(
    Normalized Index In Partition Includes One=true
    Number Of Partitions=4
    Particles Per Partition=1
    Partition Mode:ENiagara_PartitionMode_v2=::NewEnumerator1
   ) {module=PartitionParticles001}
   &PartitionParticles(Number Of Partitions=5) {module=PartitionParticles002}
   &InitializeParticle(
    Color=<rgba 1 1 1 0.8>
    Lifetime=3
    Sprite Size Mode=::NewEnumerator3
    Uniform Sprite Size=&MergeHelper_FloatSwitch(
     0=&Multiply_Float(
      A=&SelectFloatFromArray(
       Direct Array Index=$Particles.PartitionParticles.Partition
       Float Selection Array=DataInterface() {propertyValues {FloatData [35,35,35,30,30]}}
      )
      B=&MakeCustomFloatFromBool(Bool=$User.IsLargeBody, True Float=1.2, False Float=1)
     )
     1=1
     2=&MakeCustomFloatFromBool(Bool=$User.IsLargeBody, True Float=35, False Float=30)
     3=&MakeFloatFromVector2d(Vector 2D=&Multiply_Vector2D_ByFloat(
       Float=&Multiply_Float(
        A=&SelectFloatFromArray(
         Direct Array Index=$Particles.PartitionParticles002.Partition
         Float Selection Array=DataInterface() {propertyValues {FloatData [75,60,60,50,50]}}
        )
        B=&MakeCustomFloatFromBool(Bool=$User.IsLargeBody, True Float=1.1, False Float=0.95)
       )
       Vector2D=<0.9 0.9>
      ))
     Index=&SpawnGroup()
     Num=::NewEnumerator0
    )
    UsePositionOffset=false
    Write Material Random=true
   )
   &CameraOffset(Camera Offset Amount=&MergeHelper_FloatSwitch(
     0=&SelectFloatFromArray(
      Direct Array Index=$Particles.PartitionParticles.Partition
      Float Selection Array=DataInterface() {propertyValues {FloatData [6,10,10,10,10,10,10]}}
     )
     1=0
     2=-50
     3=-50
     Index=&SpawnGroup()
     Num=::NewEnumerator2
    ))
   /SetVariables(Particles.GlowCameraOffset=-30) {isSetParameters=true, module=SetVariables_#9}
   /SetVariables(Particles.GlowScale:Vector2f=&Multiply_Vector2D(
     A=<72 72>
     B=&MakeCustomVector2DFromBool(Bool=$User.IsLargeBody, True Vector2D=<1.2 1.2>, False Vector2D=<1 1>)
    )) {isSetParameters=true, module=SetVariables_#10}
   /SetVariables(Particles.GlowColor=&Multiply_LinearColorByFloat(Linear Color=<rgba 1 1 1 1>, Float=15)) {isSetParameters=true, module=SetVariables_#11}
   /SetVariables(Particles.Decal Dynamic Param:Vector4f=&MakeVector4(X=5, Y=0.1, Z=0, W=0)) {isSetParameters=true, module=SetVariables_#12}
   /SetVariables(
    Particles.DarkBacker Dynamic 0:Vector4f=<2 1 0 0>
    Particles.DarkBackerColor:LinearColor=<rgba 0.0083 0.0504 0.065 1.5>
    Particles.DarkBackerSize:Vector2f=&Multiply_Vector2D_ByFloat(
     Float=&MakeCustomFloatFromBool(Bool=$User.IsLargeBody, True Float=162, False Float=135)
     Vector2D=<1 1.1>
    )
   ) {isSetParameters=true, module=SetVariables_#13}
  ]
  particleUpdateScript [
   &ParticleState(Kill Particles When Lifetime Has Elapsed=false, Let Infinitely Lived Particles Die When Emitter Deactivates=true)
   &DynamicMaterialParameters(
    Index 0 Param 1=&Waveform("Waveform [1]"=::NewEnumerator0, "[1] Frequency"=1, "[1] Amplitude Min/Max"=<0.7 1>)
    Index 1 Param 1=$Engine.Time
    Index 1 Param 2=0.15
    Module.Index0Float0Write=true
    Module.Index0Float1Write=true
    Module.Index0Float2Write=true
    Module.Index0Float3Write=true
    Module.Index1Float0Write=true
    Module.Index1Float1Write=true
    Module.Index1Float2Write=true
    Module.Index1Float3Write=true
    Write Parameter Index 1=::NewEnumerator1
   )
   &SkeletalMeshLocation(
    Bone / Socket Index=&MergeHelper_IntSwitch(
     0=&SelectIntFromArray(Direct Array Index=$Particles.PartitionParticles.Partition, Int Selection Array=DataInterface() {propertyValues {IntData [0,1,2,3,4]}})
     1=7
     2=7
     3=7
     4=?
     5=?
     Index=&SpawnGroup()
     Num=::NewEnumerator2
    )
    Bone Sampling Mode=::NewEnumerator4
    MaskSpawnGroup=false
    Orientation Sampling=::NewEnumerator1
    Sampled Position Offset=&MergeHelper_VectorSwitch(
     0=&SelectVectorFromArray(
      Direct Array Index=$Particles.PartitionParticles.Partition
      Vector Selection Array=DataInterface() {propertyValues {FloatData [<-5 0 0>, <0 0 0>, <0 0 0>, <0 0 0>, <0 0 0>, <0 0 0>, <0 0 0>]}}
     )
     1=<0 0 0>
     2=<5 0 0>
     3=<5 0 0>
     Index=&SpawnGroup()
     Num=::NewEnumerator2
    )
    Skeletal Mesh=$System.Skeletal Mesh
   )
   /SetVariables(
    Particles.GlowColor=&MakeLinearColorFromVectorAndFloat(
     "Vector (RGB)"=&Multiply_Vector(A=$System.Color RGB, B=&MakeVectorFromLinearColorRGB(Color=$Particles.Initial.GlowColor))
     "Float (Alpha)"=&Multiply_Float(
      A=&FloatFromCurve(
       FloatCurve=DataInterface() {propertyValues {Curve {keys [
           <at 0.1 0.8 RCIM_Cubic>
           {arriveTangentWeight=0.1077, interpMode=RCIM_Cubic, leaveTangentWeight=0.0718, tangentMode=RCTM_User, time=0.4, value=1}
           {arriveTangentWeight=0.0718, interpMode=RCIM_Cubic, leaveTangentWeight=0.1077, tangentMode=RCTM_User, time=0.6, value=1}
           <at 0.9 0.8 RCIM_Cubic>
          ]}}}
       CurveIndex=Linked() {linkedVariable {name=NPC.FortniteNPC.TimeOfDay_0-1, type=NiagaraFloat}}
      )
      B=&MakeCustomFloatFromBool(Bool=$User.IsFrontEnd, True Float=0.4, False Float=1)
     )
    )
    Particles.GlowScale:Vector2f=&Multiply_Vector2D(A=$Particles.Initial.GlowScale, B=&Vector2DFromFloat(Value=1))
   ) {isSetParameters=true, module=SetVariables_#14}
   /SetVariables(Particles.RaySpriteSize:Vector2f=&Vector2DFromFloat(Value=&Multiply_Float(
      A=&Waveform("Waveform [1]"=::NewEnumerator12, "[1] Phase Scale"=0.2, "[1] Frequency"=1, "[1] Amplitude Min/Max"=<0.8 1>)
      B=&MakeCustomFloatFromBool(Bool=$User.IsLargeBody, True Float=138, False Float=115)
     ))) {isSetParameters=true, module=SetVariables_#15}
   &ScaleSpriteSize(
    Scale Factor=&MergeHelper_Vector2Switch(0=<1 1>, 1=<1 1>, 2=?, 3=?, 4=?, Index=&SpawnGroup(), Num=::NewEnumerator0)
    Scale Sprite Size Mode=::NewEnumerator2
   ) {module=ScaleSpriteSize001}
   &ScaleColor(
    Scale Alpha=&MergeHelper_FloatSwitch(
     0=&FloatFromCurve(
      FloatCurve=DataInterface() {propertyValues {Curve {keys [
          {arriveTangent=-0.0686, interpMode=RCIM_Cubic, leaveTangent=-0.0686, tangentMode=RCTM_User, time=0, value=0.9}
          {
           arriveTangent=0.0901
           arriveTangentWeight=0.0676
           interpMode=RCIM_Cubic
           leaveTangent=0.0901
           leaveTangentWeight=0.0672
           tangentMode=RCTM_User
           time=0.1964
           value=0.9848
          }
          {arriveTangentWeight=0.0651, interpMode=RCIM_Cubic, leaveTangentWeight=0.0357, tangentMode=RCTM_User, time=0.3917, value=1}
          {arriveTangentWeight=0.0357, interpMode=RCIM_Cubic, leaveTangentWeight=0.0349, tangentMode=RCTM_User, time=0.4987, value=1}
          {arriveTangentWeight=0.0349, interpMode=RCIM_Cubic, leaveTangentWeight=0.0655, tangentMode=RCTM_User, time=0.6031, value=1}
          {
           arriveTangent=-0.185
           arriveTangentWeight=0.0674
           interpMode=RCIM_Cubic
           leaveTangent=-0.185
           leaveTangentWeight=0.069
           tangentMode=RCTM_User
           time=0.7993
           value=0.9848
          }
          {arriveTangent=-0.0686, interpMode=RCIM_Cubic, leaveTangent=-0.0686, tangentMode=RCTM_User, time=1, value=0.9}
         ]}}}
      CurveIndex=Linked() {linkedVariable {name=NPC.FortniteNPC.TimeOfDay_0-1, type=NiagaraFloat}}
      Scale Curve=&MakeCustomFloatFromBool(Bool=$User.IsFrontEnd, True Float=0.6, False Float=1)
     )
     1=&FloatFromCurve(
      FloatCurve=DataInterface() {propertyValues {Curve {keys [
          {arriveTangent=-0.0686, interpMode=RCIM_Cubic, leaveTangent=-0.0686, tangentMode=RCTM_User, time=0, value=0.9}
          {
           arriveTangent=0.0901
           arriveTangentWeight=0.0676
           interpMode=RCIM_Cubic
           leaveTangent=0.0901
           leaveTangentWeight=0.0672
           tangentMode=RCTM_User
           time=0.1964
           value=0.9848
          }
          {arriveTangentWeight=0.0651, interpMode=RCIM_Cubic, leaveTangentWeight=0.0357, tangentMode=RCTM_User, time=0.3917, value=1}
          {arriveTangentWeight=0.0357, interpMode=RCIM_Cubic, leaveTangentWeight=0.0349, tangentMode=RCTM_User, time=0.4987, value=1}
          {arriveTangentWeight=0.0349, interpMode=RCIM_Cubic, leaveTangentWeight=0.0655, tangentMode=RCTM_User, time=0.6031, value=1}
          {
           arriveTangent=-0.185
           arriveTangentWeight=0.0674
           interpMode=RCIM_Cubic
           leaveTangent=-0.185
           leaveTangentWeight=0.069
           tangentMode=RCTM_User
           time=0.7993
           value=0.9848
          }
          {arriveTangent=-0.0686, interpMode=RCIM_Cubic, leaveTangent=-0.0686, tangentMode=RCTM_User, time=1, value=0.9}
         ]}}}
      CurveIndex=Linked() {linkedVariable {name=NPC.FortniteNPC.TimeOfDay_0-1, type=NiagaraFloat}}
      Scale Curve=&MakeCustomFloatFromBool(Bool=$User.IsFrontEnd, True Float=0.6, False Float=1)
     )
     2=0
     3=&MakeCustomFloatFromBool(Bool=$User.IsFrontEnd, True Float=0.9, False Float=0.9)
     Index=&SpawnGroup()
     Num=::NewEnumerator2
    )
    Scale RGB=&MergeHelper_VectorSwitch(
     0=&Multiply_VectorByFloat(Vector=$System.Color RGB, Float=10)
     1=<0 0 0>
     2=<0 0 0>
     3=&Multiply_VectorByFloat(
      Vector=$System.Color RGB
      Float=&FloatFromCurve(
       FloatCurve=DataInterface() {propertyValues {Curve {keys [
           <at 0 0.95 RCIM_Cubic>
           {arriveTangentWeight=0.0999, interpMode=RCIM_Cubic, leaveTangentWeight=0.0649, tangentMode=RCTM_User, time=0.2983, value=1}
           <at 0.4919 1 RCIM_Cubic>
           {arriveTangentWeight=0.0691, interpMode=RCIM_Cubic, leaveTangentWeight=0.1011, tangentMode=RCTM_User, time=0.6982, value=1}
           <at 1 0.95 RCIM_Cubic>
          ]}}}
       CurveIndex=Linked() {linkedVariable {name=NPC.FortniteNPC.TimeOfDay_0-1, type=NiagaraFloat}}
       Scale Curve=1.1
      )
     )
     Index=&SpawnGroup()
     Num=::NewEnumerator2
    )
    ScaleA=true
    ScaleRGB=true
   )
  ]
  rendererClasses [&NiagaraSpriteRendererProperties]
  renderers [
   {rendererClass=&NiagaraSpriteRendererProperties}
   {rendererClass=&NiagaraSpriteRendererProperties, rendererIndex=1}
   {rendererClass=&NiagaraSpriteRendererProperties, rendererIndex=2}
   {rendererClass=&NiagaraSpriteRendererProperties, rendererIndex=3}
  ]
 }
 Decal_Glows_Combined001 {
  bEnabled=false
  emitterUpdateScript [
   &EmitterState()
   &SpawnBurst_Instantaneous(Spawn Count=5, Use Spawn Probability=false, UseLoopCountLimit=true)
   &SpawnBurst_Instantaneous(Spawn Group=1, Use Spawn Probability=false, UseLoopCountLimit=true) {module=SpawnBurst_Instantaneous001}
   &SpawnBurst_Instantaneous(Spawn Group=2, Use Spawn Probability=false, UseLoopCountLimit=true) {module=SpawnBurst_Instantaneous002}
   &SpawnBurst_Instantaneous(Spawn Group=3, Use Spawn Probability=false, UseLoopCountLimit=true) {module=SpawnBurst_Instantaneous003}
  ]
  particleSpawnScript [
   /SetVariables(Particles.VisibilityTag=&SpawnGroup()) {isSetParameters=true, module=SetVariables_#8}
   &PartitionParticles(Number Of Partitions=5)
   &PartitionParticles(
    Normalized Index In Partition Includes One=true
    Number Of Partitions=4
    Particles Per Partition=1
    Partition Mode:ENiagara_PartitionMode_v2=::NewEnumerator1
   ) {module=PartitionParticles001}
   &PartitionParticles(Number Of Partitions=5) {module=PartitionParticles002}
   &InitializeParticle(
    Color=<rgba 1 1 1 0.8>
    Lifetime=3
    Sprite Size Mode=::NewEnumerator3
    Uniform Sprite Size=&MergeHelper_FloatSwitch(
     0=&Multiply_Float(
      A=&SelectFloatFromArray(
       Direct Array Index=$Particles.PartitionParticles.Partition
       Float Selection Array=DataInterface() {propertyValues {FloatData [35,35,35,30,30]}}
      )
      B=&MakeCustomFloatFromBool(Bool=$User.IsLargeBody, True Float=1.2, False Float=1)
     )
     1=1
     2=&MakeCustomFloatFromBool(Bool=$User.IsLargeBody, True Float=35, False Float=30)
     3=&MakeFloatFromVector2d(Vector 2D=&Multiply_Vector2D_ByFloat(
       Float=&Multiply_Float(
        A=&SelectFloatFromArray(
         Direct Array Index=$Particles.PartitionParticles002.Partition
         Float Selection Array=DataInterface() {propertyValues {FloatData [75,60,60,50,50]}}
        )
        B=&MakeCustomFloatFromBool(Bool=$User.IsLargeBody, True Float=1.1, False Float=0.95)
       )
       Vector2D=<0.9 0.9>
      ))
     Index=&SpawnGroup()
     Num=::NewEnumerator0
    )
    UsePositionOffset=false
    Write Material Random=true
   )
   &CameraOffset(Camera Offset Amount=&MergeHelper_FloatSwitch(
     0=&SelectFloatFromArray(
      Direct Array Index=$Particles.PartitionParticles.Partition
      Float Selection Array=DataInterface() {propertyValues {FloatData [6,10,10,10,10,10,10]}}
     )
     1=0
     2=-50
     3=-50
     Index=&SpawnGroup()
     Num=::NewEnumerator2
    ))
   /SetVariables(Particles.GlowCameraOffset=-30) {isSetParameters=true, module=SetVariables_#9}
   /SetVariables(Particles.GlowScale:Vector2f=&Multiply_Vector2D(
     A=<72 72>
     B=&MakeCustomVector2DFromBool(Bool=$User.IsLargeBody, True Vector2D=<1.2 1.2>, False Vector2D=<1 1>)
    )) {isSetParameters=true, module=SetVariables_#10}
   /SetVariables(Particles.GlowColor=&Multiply_LinearColorByFloat(Linear Color=<rgba 1 1 1 1>, Float=15)) {isSetParameters=true, module=SetVariables_#11}
   /SetVariables(Particles.Decal Dynamic Param:Vector4f=&MakeVector4(X=5, Y=0.1, Z=0, W=0)) {isSetParameters=true, module=SetVariables_#12}
   /SetVariables(
    Particles.DarkBacker Dynamic 0:Vector4f=<2 1 0 0>
    Particles.DarkBackerColor:LinearColor=<rgba 0.0028 0.0051 0.04 1.5>
    Particles.DarkBackerSize:Vector2f=&Multiply_Vector2D_ByFloat(
     Float=&MakeCustomFloatFromBool(Bool=$User.IsLargeBody, True Float=162, False Float=135)
     Vector2D=<1 1.1>
    )
   ) {isSetParameters=true, module=SetVariables_#13}
  ]
  particleUpdateScript [
   &ParticleState(Kill Particles When Lifetime Has Elapsed=false, Let Infinitely Lived Particles Die When Emitter Deactivates=true)
   &DynamicMaterialParameters(
    Index 0 Param 1=&Waveform("Waveform [1]"=::NewEnumerator0, "[1] Frequency"=1, "[1] Amplitude Min/Max"=<0.7 1>)
    Index 1 Param 1=$Engine.Time
    Index 1 Param 2=0.15
    Module.Index0Float0Write=true
    Module.Index0Float1Write=true
    Module.Index0Float2Write=true
    Module.Index0Float3Write=true
    Module.Index1Float0Write=true
    Module.Index1Float1Write=true
    Module.Index1Float2Write=true
    Module.Index1Float3Write=true
    Write Parameter Index 1=::NewEnumerator1
   )
   &SkeletalMeshLocation(
    Bone / Socket Index=&MergeHelper_IntSwitch(
     0=&SelectIntFromArray(Direct Array Index=$Particles.PartitionParticles.Partition, Int Selection Array=DataInterface() {propertyValues {IntData [0,1,2,3,4]}})
     1=7
     2=7
     3=7
     4=?
     5=?
     Index=&SpawnGroup()
     Num=::NewEnumerator2
    )
    Bone Sampling Mode=::NewEnumerator4
    MaskSpawnGroup=false
    Orientation Sampling=::NewEnumerator1
    Sampled Position Offset=&MergeHelper_VectorSwitch(
     0=&SelectVectorFromArray(
      Direct Array Index=$Particles.PartitionParticles.Partition
      Vector Selection Array=DataInterface() {propertyValues {FloatData [<-5 0 0>, <0 0 0>, <0 0 0>, <0 0 0>, <0 0 0>, <0 0 0>, <0 0 0>]}}
     )
     1=<0 0 0>
     2=<5 0 0>
     3=<5 0 0>
     Index=&SpawnGroup()
     Num=::NewEnumerator2
    )
    Skeletal Mesh=$System.Skeletal Mesh
   )
   /SetVariables(
    Particles.GlowColor=&MakeLinearColorFromVectorAndFloat(
     "Vector (RGB)"=&Multiply_Vector(
      A=&MakeVectorFromLinearColorRGB(Color=<rgba 1 0.592 0.3307 1>)
      B=&MakeVectorFromLinearColorRGB(Color=$Particles.Initial.GlowColor)
     )
     "Float (Alpha)"=&Multiply_Float(
      A=&FloatFromCurve(
       FloatCurve=DataInterface() {propertyValues {Curve {keys [
           <at 0.1 0.8 RCIM_Cubic>
           {arriveTangentWeight=0.1077, interpMode=RCIM_Cubic, leaveTangentWeight=0.0718, tangentMode=RCTM_User, time=0.4, value=1}
           {arriveTangentWeight=0.0718, interpMode=RCIM_Cubic, leaveTangentWeight=0.1077, tangentMode=RCTM_User, time=0.6, value=1}
           <at 0.9 0.8 RCIM_Cubic>
          ]}}}
       CurveIndex=Linked() {linkedVariable {name=NPC.FortniteNPC.TimeOfDay_0-1, type=NiagaraFloat}}
      )
      B=&MakeCustomFloatFromBool(Bool=$User.IsFrontEnd, True Float=0.4, False Float=1)
     )
    )
    Particles.GlowScale:Vector2f=&Multiply_Vector2D(A=$Particles.Initial.GlowScale, B=&Vector2DFromFloat(Value=1))
   ) {isSetParameters=true, module=SetVariables_#14}
   /SetVariables(Particles.RaySpriteSize:Vector2f=&Vector2DFromFloat(Value=&Multiply_Float(
      A=&Waveform("Waveform [1]"=::NewEnumerator12, "[1] Phase Scale"=0.2, "[1] Frequency"=1, "[1] Amplitude Min/Max"=<0.8 1>)
      B=&MakeCustomFloatFromBool(Bool=$User.IsLargeBody, True Float=138, False Float=115)
     ))) {isSetParameters=true, module=SetVariables_#15}
   &ScaleSpriteSize(
    Scale Factor=&MergeHelper_Vector2Switch(0=<1 1>, 1=<1 1>, 2=?, 3=?, 4=?, Index=&SpawnGroup(), Num=::NewEnumerator0)
    Scale Sprite Size Mode=::NewEnumerator2
   ) {module=ScaleSpriteSize001}
   &ScaleColor(
    Scale Alpha=&MergeHelper_FloatSwitch(
     0=&FloatFromCurve(
      FloatCurve=DataInterface() {propertyValues {Curve {keys [
          {arriveTangent=-0.0686, interpMode=RCIM_Cubic, leaveTangent=-0.0686, tangentMode=RCTM_User, time=0, value=0.9}
          {
           arriveTangent=0.0901
           arriveTangentWeight=0.0676
           interpMode=RCIM_Cubic
           leaveTangent=0.0901
           leaveTangentWeight=0.0672
           tangentMode=RCTM_User
           time=0.1964
           value=0.9848
          }
          {arriveTangentWeight=0.0651, interpMode=RCIM_Cubic, leaveTangentWeight=0.0357, tangentMode=RCTM_User, time=0.3917, value=1}
          {arriveTangentWeight=0.0357, interpMode=RCIM_Cubic, leaveTangentWeight=0.0349, tangentMode=RCTM_User, time=0.4987, value=1}
          {arriveTangentWeight=0.0349, interpMode=RCIM_Cubic, leaveTangentWeight=0.0655, tangentMode=RCTM_User, time=0.6031, value=1}
          {
           arriveTangent=-0.185
           arriveTangentWeight=0.0674
           interpMode=RCIM_Cubic
           leaveTangent=-0.185
           leaveTangentWeight=0.069
           tangentMode=RCTM_User
           time=0.7993
           value=0.9848
          }
          {arriveTangent=-0.0686, interpMode=RCIM_Cubic, leaveTangent=-0.0686, tangentMode=RCTM_User, time=1, value=0.9}
         ]}}}
      CurveIndex=Linked() {linkedVariable {name=NPC.FortniteNPC.TimeOfDay_0-1, type=NiagaraFloat}}
      Scale Curve=&MakeCustomFloatFromBool(Bool=$User.IsFrontEnd, True Float=0.6, False Float=1)
     )
     1=&FloatFromCurve(
      FloatCurve=DataInterface() {propertyValues {Curve {keys [
          {arriveTangent=-0.0686, interpMode=RCIM_Cubic, leaveTangent=-0.0686, tangentMode=RCTM_User, time=0, value=0.9}
          {
           arriveTangent=0.0901
           arriveTangentWeight=0.0676
           interpMode=RCIM_Cubic
           leaveTangent=0.0901
           leaveTangentWeight=0.0672
           tangentMode=RCTM_User
           time=0.1964
           value=0.9848
          }
          {arriveTangentWeight=0.0651, interpMode=RCIM_Cubic, leaveTangentWeight=0.0357, tangentMode=RCTM_User, time=0.3917, value=1}
          {arriveTangentWeight=0.0357, interpMode=RCIM_Cubic, leaveTangentWeight=0.0349, tangentMode=RCTM_User, time=0.4987, value=1}
          {arriveTangentWeight=0.0349, interpMode=RCIM_Cubic, leaveTangentWeight=0.0655, tangentMode=RCTM_User, time=0.6031, value=1}
          {
           arriveTangent=-0.185
           arriveTangentWeight=0.0674
           interpMode=RCIM_Cubic
           leaveTangent=-0.185
           leaveTangentWeight=0.069
           tangentMode=RCTM_User
           time=0.7993
           value=0.9848
          }
          {arriveTangent=-0.0686, interpMode=RCIM_Cubic, leaveTangent=-0.0686, tangentMode=RCTM_User, time=1, value=0.9}
         ]}}}
      CurveIndex=Linked() {linkedVariable {name=NPC.FortniteNPC.TimeOfDay_0-1, type=NiagaraFloat}}
      Scale Curve=&MakeCustomFloatFromBool(Bool=$User.IsFrontEnd, True Float=0.6, False Float=1)
     )
     2=0
     3=&MakeCustomFloatFromBool(Bool=$User.IsFrontEnd, True Float=0.9, False Float=0.9)
     Index=&SpawnGroup()
     Num=::NewEnumerator2
    )
    Scale RGB=&MergeHelper_VectorSwitch(
     0=&Multiply_VectorByFloat(Vector=&MakeVectorFromLinearColorRGB(Color=<rgba 1 0.3648 0.2135 1>), Float=10)
     1=<0 0 0>
     2=<0 0 0>
     3=&Multiply_VectorByFloat(
      Vector=$System.Color RGB
      Float=&FloatFromCurve(
       FloatCurve=DataInterface() {propertyValues {Curve {keys [
           <at 0 0.95 RCIM_Cubic>
           {arriveTangentWeight=0.0999, interpMode=RCIM_Cubic, leaveTangentWeight=0.0649, tangentMode=RCTM_User, time=0.2983, value=1}
           <at 0.4919 1 RCIM_Cubic>
           {arriveTangentWeight=0.0691, interpMode=RCIM_Cubic, leaveTangentWeight=0.1011, tangentMode=RCTM_User, time=0.6982, value=1}
           <at 1 0.95 RCIM_Cubic>
          ]}}}
       CurveIndex=Linked() {linkedVariable {name=NPC.FortniteNPC.TimeOfDay_0-1, type=NiagaraFloat}}
       Scale Curve=1.1
      )
     )
     Index=&SpawnGroup()
     Num=::NewEnumerator2
    )
    ScaleA=true
    ScaleRGB=true
   )
  ]
  rendererClasses [&NiagaraSpriteRendererProperties]
  renderers [
   {rendererClass=&NiagaraSpriteRendererProperties}
   {rendererClass=&NiagaraSpriteRendererProperties, rendererIndex=1}
   {rendererClass=&NiagaraSpriteRendererProperties, rendererIndex=2}
   {rendererClass=&NiagaraSpriteRendererProperties, rendererIndex=3}
  ]
 }
 Decal_Glows_Combined002 {
  emitterUpdateScript [
   &EmitterState()
   &SpawnBurst_Instantaneous(Spawn Count=5, Use Spawn Probability=false, UseLoopCountLimit=true)
   &SpawnBurst_Instantaneous(Spawn Group=1, Use Spawn Probability=false, UseLoopCountLimit=true) {module=SpawnBurst_Instantaneous001}
   &SpawnBurst_Instantaneous(Spawn Group=2, Use Spawn Probability=false, UseLoopCountLimit=true) {module=SpawnBurst_Instantaneous002}
   &SpawnBurst_Instantaneous(Spawn Group=3, Use Spawn Probability=false, UseLoopCountLimit=true) {module=SpawnBurst_Instantaneous003}
  ]
  particleSpawnScript [
   /SetVariables(Particles.VisibilityTag=&SpawnGroup()) {isSetParameters=true, module=SetVariables_#8}
   &PartitionParticles(Number Of Partitions=5)
   &PartitionParticles(
    Normalized Index In Partition Includes One=true
    Number Of Partitions=4
    Particles Per Partition=1
    Partition Mode:ENiagara_PartitionMode_v2=::NewEnumerator1
   ) {module=PartitionParticles001}
   &PartitionParticles(Number Of Partitions=5) {module=PartitionParticles002}
   &InitializeParticle(
    Color=<rgba 1 1 1 0.8>
    Lifetime=3
    Sprite Size Mode=::NewEnumerator3
    Uniform Sprite Size=&MergeHelper_FloatSwitch(
     0=&Multiply_Float(
      A=&SelectFloatFromArray(
       Direct Array Index=$Particles.PartitionParticles.Partition
       Float Selection Array=DataInterface() {propertyValues {FloatData [35,35,35,30,30]}}
      )
      B=&MakeCustomFloatFromBool(Bool=$User.IsLargeBody, True Float=1.2, False Float=1)
     )
     1=1
     2=&MakeCustomFloatFromBool(Bool=$User.IsLargeBody, True Float=35, False Float=30)
     3=&MakeFloatFromVector2d(Vector 2D=&Multiply_Vector2D_ByFloat(
       Float=&Multiply_Float(
        A=&SelectFloatFromArray(
         Direct Array Index=$Particles.PartitionParticles002.Partition
         Float Selection Array=DataInterface() {propertyValues {FloatData [75,60,60,50,50]}}
        )
        B=&MakeCustomFloatFromBool(Bool=$User.IsLargeBody, True Float=1.1, False Float=0.95)
       )
       Vector2D=<0.9 0.9>
      ))
     Index=&SpawnGroup()
     Num=::NewEnumerator0
    )
    UsePositionOffset=false
    Write Material Random=true
   )
   &CameraOffset(Camera Offset Amount=&MergeHelper_FloatSwitch(
     0=&SelectFloatFromArray(
      Direct Array Index=$Particles.PartitionParticles.Partition
      Float Selection Array=DataInterface() {propertyValues {FloatData [6,10,10,10,10,10,10]}}
     )
     1=0
     2=-50
     3=-50
     Index=&SpawnGroup()
     Num=::NewEnumerator2
    ))
   /SetVariables(Particles.GlowCameraOffset=-30) {isSetParameters=true, module=SetVariables_#9}
   /SetVariables(Particles.GlowScale:Vector2f=&Multiply_Vector2D(
     A=<72 72>
     B=&MakeCustomVector2DFromBool(Bool=$User.IsLargeBody, True Vector2D=<1.2 1.2>, False Vector2D=<1 1>)
    )) {isSetParameters=true, module=SetVariables_#10}
   /SetVariables(Particles.GlowColor=&Multiply_LinearColorByFloat(Linear Color=<rgba 1 1 1 1>, Float=15)) {isSetParameters=true, module=SetVariables_#11}
   /SetVariables(Particles.Decal Dynamic Param:Vector4f=&MakeVector4(X=5, Y=0.1, Z=0, W=0)) {isSetParameters=true, module=SetVariables_#12}
   /SetVariables(
    Particles.DarkBacker Dynamic 0:Vector4f=<2 1 0 0>
    Particles.DarkBackerColor:LinearColor=<rgba 0.0511 0.0153 0.065 1.5>
    Particles.DarkBackerSize:Vector2f=&Multiply_Vector2D_ByFloat(
     Float=&MakeCustomFloatFromBool(Bool=$User.IsLargeBody, True Float=162, False Float=135)
     Vector2D=<1 1.1>
    )
   ) {isSetParameters=true, module=SetVariables_#13}
  ]
  particleUpdateScript [
   &ParticleState(Kill Particles When Lifetime Has Elapsed=false, Let Infinitely Lived Particles Die When Emitter Deactivates=true)
   &DynamicMaterialParameters(
    Index 0 Param 1=&Waveform("Waveform [1]"=::NewEnumerator0, "[1] Frequency"=1, "[1] Amplitude Min/Max"=<0.7 1>)
    Index 1 Param 1=$Engine.Time
    Index 1 Param 2=0.15
    Module.Index0Float0Write=true
    Module.Index0Float1Write=true
    Module.Index0Float2Write=true
    Module.Index0Float3Write=true
    Module.Index1Float0Write=true
    Module.Index1Float1Write=true
    Module.Index1Float2Write=true
    Module.Index1Float3Write=true
    Write Parameter Index 1=::NewEnumerator1
   )
   &SkeletalMeshLocation(
    Bone / Socket Index=&MergeHelper_IntSwitch(
     0=&SelectIntFromArray(Direct Array Index=$Particles.PartitionParticles.Partition, Int Selection Array=DataInterface() {propertyValues {IntData [0,1,2,3,4]}})
     1=7
     2=7
     3=7
     4=?
     5=?
     Index=&SpawnGroup()
     Num=::NewEnumerator2
    )
    Bone Sampling Mode=::NewEnumerator4
    MaskSpawnGroup=false
    Orientation Sampling=::NewEnumerator1
    Sampled Position Offset=&MergeHelper_VectorSwitch(
     0=&SelectVectorFromArray(
      Direct Array Index=$Particles.PartitionParticles.Partition
      Vector Selection Array=DataInterface() {propertyValues {FloatData [<-5 0 0>, <0 0 0>, <0 0 0>, <0 0 0>, <0 0 0>, <0 0 0>, <0 0 0>]}}
     )
     1=<0 0 0>
     2=<5 0 0>
     3=<5 0 0>
     Index=&SpawnGroup()
     Num=::NewEnumerator2
    )
    Skeletal Mesh=$System.Skeletal Mesh
   )
   /SetVariables(
    Particles.GlowColor=&MakeLinearColorFromVectorAndFloat(
     "Vector (RGB)"=&Multiply_Vector(
      A=&MakeVectorFromLinearColorRGB(Color=<rgba 1 0.5779 0.5075 1>)
      B=&MakeVectorFromLinearColorRGB(Color=$Particles.Initial.GlowColor)
     )
     "Float (Alpha)"=&Multiply_Float(
      A=&FloatFromCurve(
       FloatCurve=DataInterface() {propertyValues {Curve {keys [
           <at 0.1 0.8 RCIM_Cubic>
           {arriveTangentWeight=0.1077, interpMode=RCIM_Cubic, leaveTangentWeight=0.0718, tangentMode=RCTM_User, time=0.4, value=1}
           {arriveTangentWeight=0.0718, interpMode=RCIM_Cubic, leaveTangentWeight=0.1077, tangentMode=RCTM_User, time=0.6, value=1}
           <at 0.9 0.8 RCIM_Cubic>
          ]}}}
       CurveIndex=Linked() {linkedVariable {name=NPC.FortniteNPC.TimeOfDay_0-1, type=NiagaraFloat}}
      )
      B=&MakeCustomFloatFromBool(Bool=$User.IsFrontEnd, True Float=0.4, False Float=1)
     )
    )
    Particles.GlowScale:Vector2f=&Multiply_Vector2D(A=$Particles.Initial.GlowScale, B=&Vector2DFromFloat(Value=1))
   ) {isSetParameters=true, module=SetVariables_#14}
   /SetVariables(Particles.RaySpriteSize:Vector2f=&Vector2DFromFloat(Value=&Multiply_Float(
      A=&Waveform("Waveform [1]"=::NewEnumerator12, "[1] Phase Scale"=0.2, "[1] Frequency"=1, "[1] Amplitude Min/Max"=<0.8 1>)
      B=&MakeCustomFloatFromBool(Bool=$User.IsLargeBody, True Float=138, False Float=115)
     ))) {isSetParameters=true, module=SetVariables_#15}
   &ScaleSpriteSize(
    Scale Factor=&MergeHelper_Vector2Switch(0=<1 1>, 1=<1 1>, 2=?, 3=?, 4=?, Index=&SpawnGroup(), Num=::NewEnumerator0)
    Scale Sprite Size Mode=::NewEnumerator2
   ) {module=ScaleSpriteSize001}
   &ScaleColor(
    Scale Alpha=&MergeHelper_FloatSwitch(
     0=&FloatFromCurve(
      FloatCurve=DataInterface() {propertyValues {Curve {keys [
          {arriveTangent=-0.0686, interpMode=RCIM_Cubic, leaveTangent=-0.0686, tangentMode=RCTM_User, time=0, value=0.9}
          {
           arriveTangent=0.0901
           arriveTangentWeight=0.0676
           interpMode=RCIM_Cubic
           leaveTangent=0.0901
           leaveTangentWeight=0.0672
           tangentMode=RCTM_User
           time=0.1964
           value=0.9848
          }
          {arriveTangentWeight=0.0651, interpMode=RCIM_Cubic, leaveTangentWeight=0.0357, tangentMode=RCTM_User, time=0.3917, value=1}
          {arriveTangentWeight=0.0357, interpMode=RCIM_Cubic, leaveTangentWeight=0.0349, tangentMode=RCTM_User, time=0.4987, value=1}
          {arriveTangentWeight=0.0349, interpMode=RCIM_Cubic, leaveTangentWeight=0.0655, tangentMode=RCTM_User, time=0.6031, value=1}
          {
           arriveTangent=-0.185
           arriveTangentWeight=0.0674
           interpMode=RCIM_Cubic
           leaveTangent=-0.185
           leaveTangentWeight=0.069
           tangentMode=RCTM_User
           time=0.7993
           value=0.9848
          }
          {arriveTangent=-0.0686, interpMode=RCIM_Cubic, leaveTangent=-0.0686, tangentMode=RCTM_User, time=1, value=0.9}
         ]}}}
      CurveIndex=Linked() {linkedVariable {name=NPC.FortniteNPC.TimeOfDay_0-1, type=NiagaraFloat}}
      Scale Curve=&MakeCustomFloatFromBool(Bool=$User.IsFrontEnd, True Float=0.6, False Float=1)
     )
     1=&FloatFromCurve(
      FloatCurve=DataInterface() {propertyValues {Curve {keys [
          {arriveTangent=-0.0686, interpMode=RCIM_Cubic, leaveTangent=-0.0686, tangentMode=RCTM_User, time=0, value=0.9}
          {
           arriveTangent=0.0901
           arriveTangentWeight=0.0676
           interpMode=RCIM_Cubic
           leaveTangent=0.0901
           leaveTangentWeight=0.0672
           tangentMode=RCTM_User
           time=0.1964
           value=0.9848
          }
          {arriveTangentWeight=0.0651, interpMode=RCIM_Cubic, leaveTangentWeight=0.0357, tangentMode=RCTM_User, time=0.3917, value=1}
          {arriveTangentWeight=0.0357, interpMode=RCIM_Cubic, leaveTangentWeight=0.0349, tangentMode=RCTM_User, time=0.4987, value=1}
          {arriveTangentWeight=0.0349, interpMode=RCIM_Cubic, leaveTangentWeight=0.0655, tangentMode=RCTM_User, time=0.6031, value=1}
          {
           arriveTangent=-0.185
           arriveTangentWeight=0.0674
           interpMode=RCIM_Cubic
           leaveTangent=-0.185
           leaveTangentWeight=0.069
           tangentMode=RCTM_User
           time=0.7993
           value=0.9848
          }
          {arriveTangent=-0.0686, interpMode=RCIM_Cubic, leaveTangent=-0.0686, tangentMode=RCTM_User, time=1, value=0.9}
         ]}}}
      CurveIndex=Linked() {linkedVariable {name=NPC.FortniteNPC.TimeOfDay_0-1, type=NiagaraFloat}}
      Scale Curve=&MakeCustomFloatFromBool(Bool=$User.IsFrontEnd, True Float=0.6, False Float=1)
     )
     2=0
     3=&MakeCustomFloatFromBool(Bool=$User.IsFrontEnd, True Float=0.9, False Float=0.9)
     Index=&SpawnGroup()
     Num=::NewEnumerator2
    )
    Scale RGB=&MergeHelper_VectorSwitch(
     0=&Multiply_VectorByFloat(Vector=&MakeVectorFromLinearColorRGB(Color=<rgba 1 0.395 0.1956 1>), Float=10)
     1=<0 0 0>
     2=<0 0 0>
     3=&Multiply_VectorByFloat(
      Vector=$System.Color RGB Palette 3
      Float=&FloatFromCurve(
       FloatCurve=DataInterface() {propertyValues {Curve {keys [
           <at 0 0.95 RCIM_Cubic>
           {arriveTangentWeight=0.0999, interpMode=RCIM_Cubic, leaveTangentWeight=0.0649, tangentMode=RCTM_User, time=0.2983, value=1}
           <at 0.4919 1 RCIM_Cubic>
           {arriveTangentWeight=0.0691, interpMode=RCIM_Cubic, leaveTangentWeight=0.1011, tangentMode=RCTM_User, time=0.6982, value=1}
           <at 1 0.95 RCIM_Cubic>
          ]}}}
       CurveIndex=Linked() {linkedVariable {name=NPC.FortniteNPC.TimeOfDay_0-1, type=NiagaraFloat}}
       Scale Curve=1.1
      )
     )
     Index=&SpawnGroup()
     Num=::NewEnumerator2
    )
    ScaleA=true
    ScaleRGB=true
   )
  ]
  rendererClasses [&NiagaraSpriteRendererProperties]
  renderers [
   {rendererClass=&NiagaraSpriteRendererProperties}
   {rendererClass=&NiagaraSpriteRendererProperties, rendererIndex=1}
   {rendererClass=&NiagaraSpriteRendererProperties, rendererIndex=2}
   {rendererClass=&NiagaraSpriteRendererProperties, rendererIndex=3}
  ]
 }
 Radial_Beams {
  bEnabled=false
  emitterUpdateScript [
   &EmitterState()
   &SpawnBurst_Instantaneous(Spawn Count=14, Use Spawn Probability=false, UseLoopCountLimit=true)
  ]
  particleSpawnScript [
   &PartitionParticles(Particles Per Partition=7, Partition Mode:ENiagara_PartitionMode_v2=::NewEnumerator1) {module=PartitionParticles001}
   &PartitionParticles(Number Of Partitions=3, Particles Per Partition=3, Partition Mode:ENiagara_PartitionMode_v2=::NewEnumerator1)
   /SetVariables(Particles.VisibilityTag=$Particles.PartitionParticles001.Partition) {isSetParameters=true, module=SetVariables_#16}
   &InitializeParticle(
    Color=<rgba 1 1 1 0.7>
    Sprite Rotation Angle=&Frac_Float(Float=&Add_Float(
      A=$Particles.PartitionParticles001.Normalized Index In Partition
      B=&MergeHelper_FloatSwitch(0=0, 1=0.2, 2=?, 3=?, Index=$Particles.PartitionParticles001.Partition, Num=::NewEnumerator0)
     ))
    Sprite Rotation Mode=::NewEnumerator1
    Sprite Size=&Multiply_Vector2D_ByFloat(
     Float=&MakeCustomFloatFromBool(Bool=$User.IsLargeBody, True Float=1.2, False Float=1)
     Vector2D=&MergeHelper_Vector2Switch(0=<40 65>, 1=<50 53>, 2=?, 3=?, 4=?, Index=$Particles.PartitionParticles001.Partition, Num=::NewEnumerator0)
    )
    Sprite Size Mode=::NewEnumerator4
    UsePositionOffset=false
    Write Material Random=true
   )
   &CameraOffset(Camera Offset Amount=-15)
  ]
  particleUpdateScript [
   &ParticleState(Kill Particles When Lifetime Has Elapsed=false)
   /SetVariables(Particles.UVIndex:NiagaraFloat=&SelectFloatFromArray(
     Array Sampling Mode=::NewEnumerator0
     Direct Array Index=?
     Float Selection Array=DataInterface() {propertyValues {FloatData [0,1,2,3]}}
     Random Seed=0
    )) {isSetParameters=true, module=SetVariables_#17}
   /SetVariables(Particles.BeamScale:NiagaraFloat=&Waveform(
     Global Input Position=$Particles.PartitionParticles.Normalized Index In Partition
     "Waveform [1]"=::NewEnumerator0
     "[1] Phase Scale"=0.2
     "[1] Frequency"=1
     "[1] Amplitude Min/Max"=<0 1>
    )) {isSetParameters=true, module=SetVariables_#18}
   &ScaleSpriteSize(
    Scale Factor=&MakeVector2D(
     X=/Game/Effects/Niagara/DynamicInputs/Math/FN_Fit(
      Clamp:NiagaraBool=false
      Value:NiagaraFloat=&Abs_Float(Float:NiagaraFloat=&OneMinusFloat(Float=$Particles.BeamScale))
      SourceMin:NiagaraFloat=0
      SourceMax:NiagaraFloat=1
      DestMin:NiagaraFloat=0.8
      DestMax:NiagaraFloat=1.6
     )
     Y=&FN_Fit(
      Clamp:NiagaraBool=false
      Value:NiagaraFloat=$Particles.BeamScale
      SourceMin:NiagaraFloat=0
      SourceMax:NiagaraFloat=1
      DestMin:NiagaraFloat=0.7
      DestMax:NiagaraFloat=1.3
     )
    )
    Scale Sprite Size Mode=::NewEnumerator2
   ) {module=ScaleSpriteSize001}
   &SkeletalMeshLocation(
    Bone / Socket Index=7
    Bone Sampling Mode=::NewEnumerator4
    MaskSpawnGroup=false
    Sampled Position Offset=<5 0 0>
    Skeletal Mesh=$System.Skeletal Mesh
   )
   &SolveForcesAndVelocity(Clamp Velocity=false, Limit Acceleration=false)
   &ScaleColor(
    Scale Alpha=&Waveform(
     Global Input Position=$Particles.PartitionParticles.Normalized Index In Partition
     "Waveform [1]"=::NewEnumerator10
     "[1] Phase Scale"=0.4
     "[1] Frequency"=1
     "[1] Amplitude Min/Max"=<0.7 1>
    )
    Scale RGB=&MakeVectorFromLinearColorRGB(Color=&ColorFromCurve(
      ColorCurve=$System.Color Curve
      CurveIndex=&Frac_Float(Float=&Add_Float(A=&Multiply_Float(A=$Engine.Time, B=0.1), B=$Particles.PartitionParticles.Normalized Index In Partition))
     ))
    ScaleA=true
    ScaleRGB=true
   )
   /SetVariables(Particles.SpriteRotation:NiagaraFloat=&Add_Float(
     A=$Particles.Initial.SpriteRotation
     B=&Waveform(
      Global Input Position=&RandomRangeFloat(Minimum=0, Maximum=1)
      "Waveform [1]"=::NewEnumerator2
      "[1] Frequency"=0.15
      "[1] Compound Frequency"=1
      "[1] Amplitude Min/Max"=<-30 30>
     )
    )) {isSetParameters=true, module=SetVariables_#19}
  ]
  rendererClasses [&NiagaraSpriteRendererProperties]
  renderers [
   {rendererClass=&NiagaraSpriteRendererProperties}
   {rendererClass=&NiagaraSpriteRendererProperties, rendererIndex=1}
  ]
 }
 Radial_Beams001 {
  bEnabled=false
  emitterUpdateScript [
   &EmitterState()
   &SpawnBurst_Instantaneous(Spawn Count=14, Use Spawn Probability=false, UseLoopCountLimit=true)
  ]
  particleSpawnScript [
   &PartitionParticles(Particles Per Partition=7, Partition Mode:ENiagara_PartitionMode_v2=::NewEnumerator1) {module=PartitionParticles001}
   &PartitionParticles(Number Of Partitions=3, Particles Per Partition=3, Partition Mode:ENiagara_PartitionMode_v2=::NewEnumerator1)
   /SetVariables(Particles.VisibilityTag=$Particles.PartitionParticles001.Partition) {isSetParameters=true, module=SetVariables_#16}
   &InitializeParticle(
    Color=<rgba 1 1 1 0.7>
    Sprite Rotation Angle=&Frac_Float(Float=&Add_Float(
      A=$Particles.PartitionParticles001.Normalized Index In Partition
      B=&MergeHelper_FloatSwitch(0=0, 1=0.2, 2=?, 3=?, Index=$Particles.PartitionParticles001.Partition, Num=::NewEnumerator0)
     ))
    Sprite Rotation Mode=::NewEnumerator1
    Sprite Size=&Multiply_Vector2D_ByFloat(
     Float=&MakeCustomFloatFromBool(Bool=$User.IsLargeBody, True Float=1.2, False Float=1)
     Vector2D=&MergeHelper_Vector2Switch(0=<40 65>, 1=<50 53>, 2=?, 3=?, 4=?, Index=$Particles.PartitionParticles001.Partition, Num=::NewEnumerator0)
    )
    Sprite Size Mode=::NewEnumerator4
    UsePositionOffset=false
    Write Material Random=true
   )
   &CameraOffset(Camera Offset Amount=-15)
  ]
  particleUpdateScript [
   &ParticleState(Kill Particles When Lifetime Has Elapsed=false)
   /SetVariables(Particles.UVIndex:NiagaraFloat=&SelectFloatFromArray(
     Array Sampling Mode=::NewEnumerator0
     Direct Array Index=?
     Float Selection Array=DataInterface() {propertyValues {FloatData [0,1,2,3]}}
     Random Seed=0
    )) {isSetParameters=true, module=SetVariables_#17}
   /SetVariables(Particles.BeamScale:NiagaraFloat=&Waveform(
     Global Input Position=$Particles.PartitionParticles.Normalized Index In Partition
     "Waveform [1]"=::NewEnumerator0
     "[1] Phase Scale"=0.2
     "[1] Frequency"=1
     "[1] Amplitude Min/Max"=<0 1>
    )) {isSetParameters=true, module=SetVariables_#18}
   &ScaleSpriteSize(
    Scale Factor=&MakeVector2D(
     X=&FN_Fit(
      Clamp:NiagaraBool=false
      Value:NiagaraFloat=&Abs_Float(Float:NiagaraFloat=&OneMinusFloat(Float=$Particles.BeamScale))
      SourceMin:NiagaraFloat=0
      SourceMax:NiagaraFloat=1
      DestMin:NiagaraFloat=0.8
      DestMax:NiagaraFloat=1.6
     )
     Y=&FN_Fit(
      Clamp:NiagaraBool=false
      Value:NiagaraFloat=$Particles.BeamScale
      SourceMin:NiagaraFloat=0
      SourceMax:NiagaraFloat=1
      DestMin:NiagaraFloat=0.7
      DestMax:NiagaraFloat=1.3
     )
    )
    Scale Sprite Size Mode=::NewEnumerator2
   ) {module=ScaleSpriteSize001}
   &SkeletalMeshLocation(
    Bone / Socket Index=7
    Bone Sampling Mode=::NewEnumerator4
    MaskSpawnGroup=false
    Sampled Position Offset=<5 0 0>
    Skeletal Mesh=$System.Skeletal Mesh
   )
   &SolveForcesAndVelocity(Clamp Velocity=false, Limit Acceleration=false)
   &ScaleColor(
    Scale Alpha=&Waveform(
     Global Input Position=$Particles.PartitionParticles.Normalized Index In Partition
     "Waveform [1]"=::NewEnumerator10
     "[1] Phase Scale"=0.4
     "[1] Frequency"=1
     "[1] Amplitude Min/Max"=<0.7 1>
    )
    Scale RGB=&MakeVectorFromLinearColorRGB(Color=&ColorFromCurve(
      ColorCurve=$System.ColorPalette2
      CurveIndex=&Frac_Float(Float=&Add_Float(A=&Multiply_Float(A=$Engine.Time, B=0.1), B=$Particles.PartitionParticles.Normalized Index In Partition))
     ))
    ScaleA=true
    ScaleRGB=true
   )
   /SetVariables(Particles.SpriteRotation:NiagaraFloat=&Add_Float(
     A=$Particles.Initial.SpriteRotation
     B=&Waveform(
      Global Input Position=&RandomRangeFloat(Minimum=0, Maximum=1)
      "Waveform [1]"=::NewEnumerator2
      "[1] Frequency"=0.15
      "[1] Compound Frequency"=1
      "[1] Amplitude Min/Max"=<-30 30>
     )
    )) {isSetParameters=true, module=SetVariables_#19}
  ]
  rendererClasses [&NiagaraSpriteRendererProperties]
  renderers [
   {rendererClass=&NiagaraSpriteRendererProperties}
   {rendererClass=&NiagaraSpriteRendererProperties, rendererIndex=1}
  ]
 }
 Radial_Beams002 {
  emitterUpdateScript [
   &EmitterState()
   &SpawnBurst_Instantaneous(Spawn Count=14, Use Spawn Probability=false, UseLoopCountLimit=true)
  ]
  particleSpawnScript [
   &PartitionParticles(Particles Per Partition=7, Partition Mode:ENiagara_PartitionMode_v2=::NewEnumerator1) {module=PartitionParticles001}
   &PartitionParticles(Number Of Partitions=3, Particles Per Partition=3, Partition Mode:ENiagara_PartitionMode_v2=::NewEnumerator1)
   /SetVariables(Particles.VisibilityTag=$Particles.PartitionParticles001.Partition) {isSetParameters=true, module=SetVariables_#16}
   &InitializeParticle(
    Color=<rgba 1 1 1 0.7>
    Sprite Rotation Angle=&Frac_Float(Float=&Add_Float(
      A=$Particles.PartitionParticles001.Normalized Index In Partition
      B=&MergeHelper_FloatSwitch(0=0, 1=0.2, 2=?, 3=?, Index=$Particles.PartitionParticles001.Partition, Num=::NewEnumerator0)
     ))
    Sprite Rotation Mode=::NewEnumerator1
    Sprite Size=&Multiply_Vector2D_ByFloat(
     Float=&MakeCustomFloatFromBool(Bool=$User.IsLargeBody, True Float=1.2, False Float=1)
     Vector2D=&MergeHelper_Vector2Switch(0=<40 65>, 1=<50 53>, 2=?, 3=?, 4=?, Index=$Particles.PartitionParticles001.Partition, Num=::NewEnumerator0)
    )
    Sprite Size Mode=::NewEnumerator4
    UsePositionOffset=false
    Write Material Random=true
   )
   &CameraOffset(Camera Offset Amount=-15)
  ]
  particleUpdateScript [
   &ParticleState(Kill Particles When Lifetime Has Elapsed=false)
   /SetVariables(Particles.UVIndex:NiagaraFloat=&SelectFloatFromArray(
     Array Sampling Mode=::NewEnumerator0
     Direct Array Index=?
     Float Selection Array=DataInterface() {propertyValues {FloatData [0,1,2,3]}}
     Random Seed=0
    )) {isSetParameters=true, module=SetVariables_#17}
   /SetVariables(Particles.BeamScale:NiagaraFloat=&Waveform(
     Global Input Position=$Particles.PartitionParticles.Normalized Index In Partition
     "Waveform [1]"=::NewEnumerator0
     "[1] Phase Scale"=0.2
     "[1] Frequency"=1
     "[1] Amplitude Min/Max"=<0 1>
    )) {isSetParameters=true, module=SetVariables_#18}
   &ScaleSpriteSize(
    Scale Factor=&MakeVector2D(
     X=&FN_Fit(
      Clamp:NiagaraBool=false
      Value:NiagaraFloat=&Abs_Float(Float:NiagaraFloat=&OneMinusFloat(Float=$Particles.BeamScale))
      SourceMin:NiagaraFloat=0
      SourceMax:NiagaraFloat=1
      DestMin:NiagaraFloat=0.8
      DestMax:NiagaraFloat=1.6
     )
     Y=&FN_Fit(
      Clamp:NiagaraBool=false
      Value:NiagaraFloat=$Particles.BeamScale
      SourceMin:NiagaraFloat=0
      SourceMax:NiagaraFloat=1
      DestMin:NiagaraFloat=0.7
      DestMax:NiagaraFloat=1.3
     )
    )
    Scale Sprite Size Mode=::NewEnumerator2
   ) {module=ScaleSpriteSize001}
   &SkeletalMeshLocation(
    Bone / Socket Index=7
    Bone Sampling Mode=::NewEnumerator4
    MaskSpawnGroup=false
    Sampled Position Offset=<5 0 0>
    Skeletal Mesh=$System.Skeletal Mesh
   )
   &SolveForcesAndVelocity(Clamp Velocity=false, Limit Acceleration=false)
   &ScaleColor(
    Scale Alpha=&Waveform(
     Global Input Position=$Particles.PartitionParticles.Normalized Index In Partition
     "Waveform [1]"=::NewEnumerator10
     "[1] Phase Scale"=0.4
     "[1] Frequency"=1
     "[1] Amplitude Min/Max"=<0.7 1>
    )
    Scale RGB=&MakeVectorFromLinearColorRGB(Color=&ColorFromCurve(
      ColorCurve=$System.ColorPalette3
      CurveIndex=&Frac_Float(Float=&Add_Float(A=&Multiply_Float(A=$Engine.Time, B=0.1), B=$Particles.PartitionParticles.Normalized Index In Partition))
     ))
    ScaleA=true
    ScaleRGB=true
   )
   /SetVariables(Particles.SpriteRotation:NiagaraFloat=&Add_Float(
     A=$Particles.Initial.SpriteRotation
     B=&Waveform(
      Global Input Position=&RandomRangeFloat(Minimum=0, Maximum=1)
      "Waveform [1]"=::NewEnumerator2
      "[1] Frequency"=0.15
      "[1] Compound Frequency"=1
      "[1] Amplitude Min/Max"=<-30 30>
     )
    )) {isSetParameters=true, module=SetVariables_#19}
  ]
  rendererClasses [&NiagaraSpriteRendererProperties]
  renderers [
   {rendererClass=&NiagaraSpriteRendererProperties}
   {rendererClass=&NiagaraSpriteRendererProperties, rendererIndex=1}
  ]
 }
 SpaceDust_Clouds {
  bEnabled=false
  emitterUpdateScript [
   &EmitterState()
   &SpawnBurst_Instantaneous(Spawn Count=7, Use Spawn Probability=false, UseLoopCountLimit=true)
   &SpawnRate(SpawnRate=8, Use Spawn Probability=false)
   &SpawnBurst_Instantaneous(Spawn Count=2, Spawn Group=1, Use Spawn Probability=false, UseLoopCountLimit=true) {module=SpawnBurst_Instantaneous001}
   &SpawnRate(SpawnGroup=1, SpawnRate=2.5, Use Spawn Probability=false) {module=SpawnRate001}
  ]
  particleSpawnScript [
   &PartitionParticles(Number Of Partitions=3, Particles Per Partition=4, Partition Mode:ENiagara_PartitionMode_v2=::NewEnumerator1)
   /SetVariables(Particles.VisibilityTag=&MergeHelper_IntSwitch(
     0=&RandomRangeInt(Minimum=0, Maximum=1)
     1=2
     2=?
     3=?
     4=?
     5=?
     Index=&SpawnGroup()
     Num=::NewEnumerator0
    )) {isSetParameters=true, module=SetVariables_#20}
   &InitializeParticle(
    Color=&MergeHelper_ColorSwitch(0=<rgba 1 1 1 0.35>, 1=<rgba 1.1 1.1 1.1 0.7>, 2=<rgba 1 1 1 0.9>, 3=?, Index=$Particles.VisibilityTag, Num=::NewEnumerator1)
    Lifetime Max=1.75
    Lifetime Min=1.25
    Lifetime Mode=::NewEnumerator1
    Sprite Rotation Mode=::NewEnumerator2
    Sprite Size Mode=::NewEnumerator3
    Sprite UV Mode=::NewEnumerator3
    Uniform Sprite Size=&MergeHelper_FloatSwitch(
     0=&Multiply_Float(
      A=&RandomRangeFloat(Minimum=25, Maximum=35)
      B=&MakeCustomFloatFromBool(Bool=$User.IsLargeBody, True Float=1.2, False Float=1)
     )
     1=&Multiply_Float(
      A=&RandomRangeFloat(Minimum=93, Maximum=98)
      B=&MakeCustomFloatFromBool(Bool=$User.IsLargeBody, True Float=1.15, False Float=1)
     )
     2=?
     3=?
     Index=&SpawnGroup()
     Num=::NewEnumerator0
    )
    UsePositionOffset=false
    Write Material Random=true
   )
   &CameraOffset(Camera Offset Amount=&MergeHelper_FloatSwitch(0=-25, 1=-20, 2=?, 3=?, Index=&SpawnGroup(), Num=::NewEnumerator0))
   &ShapeLocation(
    Hemisphere Distribution=&MergeHelper_Vector2Switch(0=<1 0.8>, 1=<1 1>, 2=?, 3=?, 4=?, Index=&SpawnGroup(), Num=::NewEnumerator0)
    MaskSpawnGroup=false
    Override Seed=false
    Shape Origin=<0 0 0>
    Sphere Radius=&MergeHelper_FloatSwitch(
     0=&MakeCustomFloatFromBool(Bool=$User.IsLargeBody, True Float=20, False Float=18)
     1=&MakeCustomFloatFromBool(Bool=$User.IsLargeBody, True Float=5, False Float=4)
     2=?
     3=?
     Index=&SpawnGroup()
     Num=::NewEnumerator0
    )
    Sphere Surface Distribution=1
    UseOwnerScale=true
   )
   &AddVelocity(
    Module.Use Falloff Distance=false
    Origin Offset=<0 0 -10>
    Origin Offset Coordinate Space=::Simulation
    Override Seed=false
    Velocity Mode=::NewEnumerator1
    Velocity Origin=<0 0 0>
    Velocity Speed=&MergeHelper_FloatSwitch(
     0=&RandomRangeFloat(Minimum=10, Maximum=20)
     1=0
     2=?
     3=?
     Index=&SpawnGroup()
     Num=::NewEnumerator0
    )
   ) {module=AddVelocity001}
  ]
  particleUpdateScript [
   &ParticleState()
   &ScaleSpriteSize(
    Scale Sprite Size Mode=::NewEnumerator0
    Uniform Scale Factor=&MergeHelper_FloatSwitch(
     0=&FloatFromCurve(FloatCurve=DataInterface() {propertyValues {Curve {
         defaultValue=1
         keys [
          {arriveTangent=1.5038, interpMode=RCIM_Cubic, leaveTangent=1.5038, tangentMode=RCTM_User, time=0, value=0.75}
          {arriveTangent=0.2735, interpMode=RCIM_Cubic, leaveTangent=0.2735, time=0.1963, value=0.8713}
          {arriveTangent=0.3468, interpMode=RCIM_Cubic, leaveTangent=0.3468, time=0.7984, value=0.9684}
          {arriveTangent=1.8818, interpMode=RCIM_Cubic, leaveTangent=1.8818, tangentMode=RCTM_User, time=1, value=1.15}
         ]
        }}})
     1=&FloatFromCurve(FloatCurve=DataInterface() {propertyValues {Curve {
         defaultValue=1
         keys [
          {arriveTangent=1.5038, interpMode=RCIM_Cubic, leaveTangent=1.5038, tangentMode=RCTM_User, time=0, value=0.75}
          {arriveTangent=0.2205, interpMode=RCIM_Cubic, leaveTangent=0.2205, time=0.1963, value=0.8713}
          {arriveTangent=0.1601, interpMode=RCIM_Cubic, leaveTangent=0.1601, time=0.7984, value=0.926}
          {arriveTangent=1.8818, interpMode=RCIM_Cubic, leaveTangent=1.8818, tangentMode=RCTM_User, time=1, value=1}
         ]
        }}})
     2=?
     3=?
     Index=&SpawnGroup()
     Num=::NewEnumerator0
    )
   )
   &SpriteRotationRate(Rotation Rate=&MergeHelper_FloatSwitch(
     0=30
     1=&Multiply_Float(
      A=&RandomRangeFloat(Minimum=10, Maximum=25)
      B=&SelectFloatFromArray(
       Array Sampling Mode=::NewEnumerator0
       Direct Array Index=?
       Float Selection Array=DataInterface() {propertyValues {FloatData [-1,1]}}
       Random Seed=0
      )
     )
     2=?
     3=?
     Index=&SpawnGroup()
     Num=::NewEnumerator0
    ))
   &ScaleColor(
    Scale Alpha=&MergeHelper_FloatSwitch(
     0=&Multiply_Float(
      A=&FloatFromCurve(
       FloatCurve=DataInterface() {propertyValues {Curve {keys [
           {interpMode=RCIM_Cubic, leaveTangentWeight=0.2338, tangentMode=RCTM_Break, time=0, value=0}
           <at 0.3 1 RCIM_Cubic>
           {arriveTangentWeight=0.3188, interpMode=RCIM_Cubic, tangentMode=RCTM_User, time=1, value=0}
          ]}}}
       Scale Curve=&MakeCustomFloatFromBool(Bool=$User.IsFrontEnd, True Float=0.7, False Float=1)
      )
      B=&FloatFromCurve(
       FloatCurve=DataInterface() {propertyValues {Curve {keys [
           <at 0.1 0.85 RCIM_Cubic>
           {arriveTangentWeight=0.1077, interpMode=RCIM_Cubic, leaveTangentWeight=0.0718, tangentMode=RCTM_User, time=0.4, value=1}
           {arriveTangentWeight=0.0718, interpMode=RCIM_Cubic, leaveTangentWeight=0.1077, tangentMode=RCTM_User, time=0.6, value=1}
           <at 0.9 0.85 RCIM_Cubic>
          ]}}}
       CurveIndex=Linked() {linkedVariable {name=NPC.FortniteNPC.TimeOfDay_0-1, type=NiagaraFloat}}
      )
     )
     1=&Multiply_Float(
      A=&FloatFromCurve(
       FloatCurve=DataInterface() {propertyValues {Curve {keys [
           {interpMode=RCIM_Cubic, leaveTangent=1.1717, leaveTangentWeight=0.2338, tangentMode=RCTM_Break, time=0, value=0}
           <at 0.3 1 RCIM_Cubic>
           {arriveTangent=-0.8056, arriveTangentWeight=0.3188, interpMode=RCIM_Cubic, leaveTangent=-0.8056, tangentMode=RCTM_User, time=1, value=0}
          ]}}}
       Scale Curve=&MakeCustomFloatFromBool(Bool=$User.IsFrontEnd, True Float=1.1, False Float=1)
      )
      B=&FloatFromCurve(
       FloatCurve=DataInterface() {propertyValues {Curve {keys [
           <at 0.1 0.95 RCIM_Cubic>
           {arriveTangentWeight=0.1077, interpMode=RCIM_Cubic, leaveTangentWeight=0.0718, tangentMode=RCTM_User, time=0.4, value=1}
           {arriveTangentWeight=0.0718, interpMode=RCIM_Cubic, leaveTangentWeight=0.1077, tangentMode=RCTM_User, time=0.6, value=1}
           <at 0.9 0.95 RCIM_Cubic>
          ]}}}
       CurveIndex=Linked() {linkedVariable {name=NPC.FortniteNPC.TimeOfDay_0-1, type=NiagaraFloat}}
      )
     )
     2=?
     3=?
     Index=&SpawnGroup()
     Num=::NewEnumerator0
    )
    Scale RGB=&MergeHelper_VectorSwitch(
     0=&Multiply_Vector(
      A=&MakeVectorFromLinearColorRGB(Color=&ColorFromCurve(ColorCurve=DataInterface() {propertyValues {
          AlphaCurve {keys [<at 0 1 RCIM_Cubic>]}
          BlueCurve {keys [<at 0 1>,<at 0.331 1.5 RCIM_Cubic>,<at 1 1 RCIM_Cubic>]}
          GreenCurve {keys [<at 0 0.8441>, {arriveTangent=0.1417, interpMode=RCIM_Cubic, leaveTangent=0.1417, time=0.331, value=1.5}, <at 1 0.9857 RCIM_Cubic>]}
          RedCurve {keys [<at 0 0.5318>, {arriveTangent=-0.5318, interpMode=RCIM_Cubic, leaveTangent=-0.5318, time=0.331, value=1.5}, <at 1 0 RCIM_Cubic>]}
         }}))
      B=&MakeVectorFromLinearColorRGB(Color=&ColorFromCurve(
        ColorCurve=$System.Color Curve Desaturated
        CurveIndex=&Frac_Float(Float=&Add_Float(A=$Particles.NormalizedAge, B=&RandomRangeFloat(Minimum=0, Maximum=1)))
       ))
     )
     1=&Multiply_Vector(
      A=&MakeVectorFromLinearColorRGB(Color=&ColorFromCurve(ColorCurve=DataInterface() {propertyValues {
          AlphaCurve {keys [<at 0 1 RCIM_Cubic>]}
          BlueCurve {keys [<at 0 1>,<at 0.331 1.7 RCIM_Cubic>,<at 1 1 RCIM_Cubic>]}
          GreenCurve {keys [<at 0 0.8441>, {arriveTangent=0.1417, interpMode=RCIM_Cubic, leaveTangent=0.1417, time=0.331, value=1.7}, <at 1 0.9857 RCIM_Cubic>]}
          RedCurve {keys [<at 0 0.5318>, {arriveTangent=-0.5318, interpMode=RCIM_Cubic, leaveTangent=-0.5318, time=0.331, value=1.7}, <at 1 0 RCIM_Cubic>]}
         }}))
      B=&MakeVectorFromLinearColorRGB(Color=&ColorFromCurve(
        ColorCurve=$System.Color Curve Desaturated
        CurveIndex=&Frac_Float(Float=&Add_Float(A=$Particles.NormalizedAge, B=&RandomRangeFloat(Minimum=0, Maximum=1)))
       ))
     )
     2=?
     3=?
     Index=&SpawnGroup()
     Num=::NewEnumerator0
    )
    ScaleA=true
    ScaleRGB=true
   ) {module=ScaleColor001}
   &DynamicMaterialParameters(
    Index 0 Param 1=&RandomRangeFloat(Minimum=0, Maximum=3)
    Module.Index0Float0Write=true
    Module.Index0Float1Write=true
    Module.Index0Float2Write=true
    Module.Index0Float3Write=true
   )
   &Drag(Drag=0.5, Use Linear Drag=true, Use Rotational Drag=true)
   &SolveForcesAndVelocity(Clamp Velocity=false, Limit Acceleration=false)
   /SetVariables(Particles.PositionMod:NiagaraPosition=&AddVectorToPosition(
     Position=$Engine.Emitter.SimulationPosition
     Vector=&Add_Vector(
      A=&TransformVector(Vector=Linked() {linkedVariable {name=System.CenteredPos, type=NiagaraPosition}}, Destination Space=::Simulation)
      B=Linked() {linkedVariable {name=Particles.Position, type=NiagaraPosition}}
     )
    )) {isSetParameters=true, module=SetVariables_#21}
  ]
  rendererClasses [&NiagaraSpriteRendererProperties]
  renderers [
   {rendererClass=&NiagaraSpriteRendererProperties}
   {rendererClass=&NiagaraSpriteRendererProperties, rendererIndex=1}
   {rendererClass=&NiagaraSpriteRendererProperties, rendererIndex=2}
  ]
 }
 SpaceDust_Clouds001 {
  bEnabled=false
  emitterUpdateScript [
   &EmitterState()
   &SpawnBurst_Instantaneous(Spawn Count=7, Use Spawn Probability=false, UseLoopCountLimit=true)
   &SpawnRate(SpawnRate=8, Use Spawn Probability=false)
   &SpawnBurst_Instantaneous(Spawn Count=2, Spawn Group=1, Use Spawn Probability=false, UseLoopCountLimit=true) {module=SpawnBurst_Instantaneous001}
   &SpawnRate(SpawnGroup=1, SpawnRate=2.5, Use Spawn Probability=false) {module=SpawnRate001}
  ]
  particleSpawnScript [
   &PartitionParticles(Number Of Partitions=3, Particles Per Partition=4, Partition Mode:ENiagara_PartitionMode_v2=::NewEnumerator1)
   /SetVariables(Particles.VisibilityTag=&MergeHelper_IntSwitch(
     0=&RandomRangeInt(Minimum=0, Maximum=1)
     1=2
     2=?
     3=?
     4=?
     5=?
     Index=&SpawnGroup()
     Num=::NewEnumerator0
    )) {isSetParameters=true, module=SetVariables_#20}
   &InitializeParticle(
    Color=&MergeHelper_ColorSwitch(0=<rgba 1 1 1 0.35>, 1=<rgba 1.1 1.1 1.1 0.7>, 2=<rgba 1 1 1 0.9>, 3=?, Index=$Particles.VisibilityTag, Num=::NewEnumerator1)
    Lifetime Max=1.75
    Lifetime Min=1.25
    Lifetime Mode=::NewEnumerator1
    Sprite Rotation Mode=::NewEnumerator2
    Sprite Size Mode=::NewEnumerator3
    Sprite UV Mode=::NewEnumerator3
    Uniform Sprite Size=&MergeHelper_FloatSwitch(
     0=&Multiply_Float(
      A=&RandomRangeFloat(Minimum=25, Maximum=35)
      B=&MakeCustomFloatFromBool(Bool=$User.IsLargeBody, True Float=1.2, False Float=1)
     )
     1=&Multiply_Float(
      A=&RandomRangeFloat(Minimum=93, Maximum=98)
      B=&MakeCustomFloatFromBool(Bool=$User.IsLargeBody, True Float=1.15, False Float=1)
     )
     2=?
     3=?
     Index=&SpawnGroup()
     Num=::NewEnumerator0
    )
    UsePositionOffset=false
    Write Material Random=true
   )
   &CameraOffset(Camera Offset Amount=&MergeHelper_FloatSwitch(0=-25, 1=-20, 2=?, 3=?, Index=&SpawnGroup(), Num=::NewEnumerator0))
   &ShapeLocation(
    Hemisphere Distribution=&MergeHelper_Vector2Switch(0=<1 0.8>, 1=<1 1>, 2=?, 3=?, 4=?, Index=&SpawnGroup(), Num=::NewEnumerator0)
    MaskSpawnGroup=false
    Override Seed=false
    Shape Origin=<0 0 0>
    Sphere Radius=&MergeHelper_FloatSwitch(
     0=&MakeCustomFloatFromBool(Bool=$User.IsLargeBody, True Float=20, False Float=18)
     1=&MakeCustomFloatFromBool(Bool=$User.IsLargeBody, True Float=5, False Float=4)
     2=?
     3=?
     Index=&SpawnGroup()
     Num=::NewEnumerator0
    )
    Sphere Surface Distribution=1
    UseOwnerScale=true
   )
   &AddVelocity(
    Module.Use Falloff Distance=false
    Origin Offset=<0 0 -10>
    Origin Offset Coordinate Space=::Simulation
    Override Seed=false
    Velocity Mode=::NewEnumerator1
    Velocity Origin=<0 0 0>
    Velocity Speed=&MergeHelper_FloatSwitch(
     0=&RandomRangeFloat(Minimum=10, Maximum=20)
     1=0
     2=?
     3=?
     Index=&SpawnGroup()
     Num=::NewEnumerator0
    )
   ) {module=AddVelocity001}
  ]
  particleUpdateScript [
   &ParticleState()
   &ScaleSpriteSize(
    Scale Sprite Size Mode=::NewEnumerator0
    Uniform Scale Factor=&MergeHelper_FloatSwitch(
     0=&FloatFromCurve(FloatCurve=DataInterface() {propertyValues {Curve {
         defaultValue=1
         keys [
          {arriveTangent=1.5038, interpMode=RCIM_Cubic, leaveTangent=1.5038, tangentMode=RCTM_User, time=0, value=0.75}
          {arriveTangent=0.2735, interpMode=RCIM_Cubic, leaveTangent=0.2735, time=0.1963, value=0.8713}
          {arriveTangent=0.3468, interpMode=RCIM_Cubic, leaveTangent=0.3468, time=0.7984, value=0.9684}
          {arriveTangent=1.8818, interpMode=RCIM_Cubic, leaveTangent=1.8818, tangentMode=RCTM_User, time=1, value=1.15}
         ]
        }}})
     1=&FloatFromCurve(FloatCurve=DataInterface() {propertyValues {Curve {
         defaultValue=1
         keys [
          {arriveTangent=1.5038, interpMode=RCIM_Cubic, leaveTangent=1.5038, tangentMode=RCTM_User, time=0, value=0.75}
          {arriveTangent=0.2205, interpMode=RCIM_Cubic, leaveTangent=0.2205, time=0.1963, value=0.8713}
          {arriveTangent=0.1601, interpMode=RCIM_Cubic, leaveTangent=0.1601, time=0.7984, value=0.926}
          {arriveTangent=1.8818, interpMode=RCIM_Cubic, leaveTangent=1.8818, tangentMode=RCTM_User, time=1, value=1}
         ]
        }}})
     2=?
     3=?
     Index=&SpawnGroup()
     Num=::NewEnumerator0
    )
   )
   &SpriteRotationRate(Rotation Rate=&MergeHelper_FloatSwitch(
     0=30
     1=&Multiply_Float(
      A=&RandomRangeFloat(Minimum=10, Maximum=25)
      B=&SelectFloatFromArray(
       Array Sampling Mode=::NewEnumerator0
       Direct Array Index=?
       Float Selection Array=DataInterface() {propertyValues {FloatData [-1,1]}}
       Random Seed=0
      )
     )
     2=?
     3=?
     Index=&SpawnGroup()
     Num=::NewEnumerator0
    ))
   &ScaleColor(
    Scale Alpha=&MergeHelper_FloatSwitch(
     0=&Multiply_Float(
      A=&FloatFromCurve(
       FloatCurve=DataInterface() {propertyValues {Curve {keys [
           {interpMode=RCIM_Cubic, leaveTangentWeight=0.2338, tangentMode=RCTM_Break, time=0, value=0}
           <at 0.3 1 RCIM_Cubic>
           {arriveTangentWeight=0.3188, interpMode=RCIM_Cubic, tangentMode=RCTM_User, time=1, value=0}
          ]}}}
       Scale Curve=&MakeCustomFloatFromBool(Bool=$User.IsFrontEnd, True Float=0.7, False Float=1)
      )
      B=&FloatFromCurve(
       FloatCurve=DataInterface() {propertyValues {Curve {keys [
           <at 0.1 0.85 RCIM_Cubic>
           {arriveTangentWeight=0.1077, interpMode=RCIM_Cubic, leaveTangentWeight=0.0718, tangentMode=RCTM_User, time=0.4, value=1}
           {arriveTangentWeight=0.0718, interpMode=RCIM_Cubic, leaveTangentWeight=0.1077, tangentMode=RCTM_User, time=0.6, value=1}
           <at 0.9 0.85 RCIM_Cubic>
          ]}}}
       CurveIndex=Linked() {linkedVariable {name=NPC.FortniteNPC.TimeOfDay_0-1, type=NiagaraFloat}}
      )
     )
     1=&Multiply_Float(
      A=&FloatFromCurve(
       FloatCurve=DataInterface() {propertyValues {Curve {keys [
           {interpMode=RCIM_Cubic, leaveTangent=1.1717, leaveTangentWeight=0.2338, tangentMode=RCTM_Break, time=0, value=0}
           <at 0.3 1 RCIM_Cubic>
           {arriveTangent=-0.8056, arriveTangentWeight=0.3188, interpMode=RCIM_Cubic, leaveTangent=-0.8056, tangentMode=RCTM_User, time=1, value=0}
          ]}}}
       Scale Curve=&MakeCustomFloatFromBool(Bool=$User.IsFrontEnd, True Float=1.1, False Float=1)
      )
      B=&FloatFromCurve(
       FloatCurve=DataInterface() {propertyValues {Curve {keys [
           <at 0.1 0.95 RCIM_Cubic>
           {arriveTangentWeight=0.1077, interpMode=RCIM_Cubic, leaveTangentWeight=0.0718, tangentMode=RCTM_User, time=0.4, value=1}
           {arriveTangentWeight=0.0718, interpMode=RCIM_Cubic, leaveTangentWeight=0.1077, tangentMode=RCTM_User, time=0.6, value=1}
           <at 0.9 0.95 RCIM_Cubic>
          ]}}}
       CurveIndex=Linked() {linkedVariable {name=NPC.FortniteNPC.TimeOfDay_0-1, type=NiagaraFloat}}
      )
     )
     2=?
     3=?
     Index=&SpawnGroup()
     Num=::NewEnumerator0
    )
    Scale RGB=&MergeHelper_VectorSwitch(
     0=&Multiply_Vector(
      A=&MakeVectorFromLinearColorRGB(Color=&ColorFromCurve(ColorCurve=DataInterface() {propertyValues {
          AlphaCurve {keys [<at 0 1 RCIM_Cubic>]}
          BlueCurve {keys [<at 0 0.4259>, {arriveTangent=0.3603, interpMode=RCIM_Cubic, leaveTangent=0.3603, time=0.331, value=1.5}, <at 1 0.7862 RCIM_Cubic>]}
          GreenCurve {keys [<at 0 0.6159>, {arriveTangent=0.3841, interpMode=RCIM_Cubic, leaveTangent=0.3841, time=0.331, value=1.5}, <at 1 1 RCIM_Cubic>]}
          RedCurve {keys [<at 0 1>, {arriveTangent=-0.6901, interpMode=RCIM_Cubic, leaveTangent=-0.6901, time=0.331, value=1.5}, <at 1 0.3099 RCIM_Cubic>]}
         }}))
      B=&MakeVectorFromLinearColorRGB(Color=&ColorFromCurve(
        ColorCurve=$System.ColorPalette2 Desaturated
        CurveIndex=&Frac_Float(Float=&Add_Float(A=$Particles.NormalizedAge, B=&RandomRangeFloat(Minimum=0, Maximum=1)))
       ))
     )
     1=&Multiply_Vector(
      A=&MakeVectorFromLinearColorRGB(Color=&ColorFromCurve(ColorCurve=DataInterface() {propertyValues {
          AlphaCurve {keys [<at 0 1 RCIM_Cubic>]}
          BlueCurve {keys [<at 0 0.4259>, {arriveTangent=0.3603, interpMode=RCIM_Cubic, leaveTangent=0.3603, time=0.331, value=1.5}, <at 1 0.7862 RCIM_Cubic>]}
          GreenCurve {keys [<at 0 0.6159>, {arriveTangent=0.3841, interpMode=RCIM_Cubic, leaveTangent=0.3841, time=0.331, value=1.5}, <at 1 1 RCIM_Cubic>]}
          RedCurve {keys [<at 0 1>, {arriveTangent=-0.6901, interpMode=RCIM_Cubic, leaveTangent=-0.6901, time=0.331, value=1.5}, <at 1 0.3099 RCIM_Cubic>]}
         }}))
      B=&MakeVectorFromLinearColorRGB(Color=&ColorFromCurve(
        ColorCurve=$System.ColorPalette2 Desaturated
        CurveIndex=&Frac_Float(Float=&Add_Float(A=$Particles.NormalizedAge, B=&RandomRangeFloat(Minimum=0, Maximum=1)))
       ))
     )
     2=?
     3=?
     Index=&SpawnGroup()
     Num=::NewEnumerator0
    )
    ScaleA=true
    ScaleRGB=true
   ) {module=ScaleColor001}
   &DynamicMaterialParameters(
    Index 0 Param 1=&RandomRangeFloat(Minimum=0, Maximum=3)
    Module.Index0Float0Write=true
    Module.Index0Float1Write=true
    Module.Index0Float2Write=true
    Module.Index0Float3Write=true
   )
   &Drag(Drag=0.5, Use Linear Drag=true, Use Rotational Drag=true)
   &SolveForcesAndVelocity(Clamp Velocity=false, Limit Acceleration=false)
   /SetVariables(Particles.PositionMod:NiagaraPosition=&AddVectorToPosition(
     Position=$Engine.Emitter.SimulationPosition
     Vector=&Add_Vector(
      A=&TransformVector(Vector=Linked() {linkedVariable {name=System.CenteredPos, type=NiagaraPosition}}, Destination Space=::Simulation)
      B=Linked() {linkedVariable {name=Particles.Position, type=NiagaraPosition}}
     )
    )) {isSetParameters=true, module=SetVariables_#21}
  ]
  rendererClasses [&NiagaraSpriteRendererProperties]
  renderers [
   {rendererClass=&NiagaraSpriteRendererProperties}
   {rendererClass=&NiagaraSpriteRendererProperties, rendererIndex=1}
   {rendererClass=&NiagaraSpriteRendererProperties, rendererIndex=2}
  ]
 }
 SpaceDust_Clouds002 {
  emitterUpdateScript [
   &EmitterState()
   &SpawnBurst_Instantaneous(Spawn Count=7, Use Spawn Probability=false, UseLoopCountLimit=true)
   &SpawnRate(SpawnRate=8, Use Spawn Probability=false)
   &SpawnBurst_Instantaneous(Spawn Count=2, Spawn Group=1, Use Spawn Probability=false, UseLoopCountLimit=true) {module=SpawnBurst_Instantaneous001}
   &SpawnRate(SpawnGroup=1, SpawnRate=2.5, Use Spawn Probability=false) {module=SpawnRate001}
  ]
  particleSpawnScript [
   &PartitionParticles(Number Of Partitions=3, Particles Per Partition=4, Partition Mode:ENiagara_PartitionMode_v2=::NewEnumerator1)
   /SetVariables(Particles.VisibilityTag=&MergeHelper_IntSwitch(
     0=&RandomRangeInt(Minimum=0, Maximum=1)
     1=2
     2=?
     3=?
     4=?
     5=?
     Index=&SpawnGroup()
     Num=::NewEnumerator0
    )) {isSetParameters=true, module=SetVariables_#20}
   &InitializeParticle(
    Color=&MergeHelper_ColorSwitch(0=<rgba 1 1 1 0.35>, 1=<rgba 1.1 1.1 1.1 0.7>, 2=<rgba 1 1 1 0.9>, 3=?, Index=$Particles.VisibilityTag, Num=::NewEnumerator1)
    Lifetime Max=1.75
    Lifetime Min=1.25
    Lifetime Mode=::NewEnumerator1
    Sprite Rotation Mode=::NewEnumerator2
    Sprite Size Mode=::NewEnumerator3
    Sprite UV Mode=::NewEnumerator3
    Uniform Sprite Size=&MergeHelper_FloatSwitch(
     0=&Multiply_Float(
      A=&RandomRangeFloat(Minimum=25, Maximum=35)
      B=&MakeCustomFloatFromBool(Bool=$User.IsLargeBody, True Float=1.2, False Float=1)
     )
     1=&Multiply_Float(
      A=&RandomRangeFloat(Minimum=93, Maximum=98)
      B=&MakeCustomFloatFromBool(Bool=$User.IsLargeBody, True Float=1.15, False Float=1)
     )
     2=?
     3=?
     Index=&SpawnGroup()
     Num=::NewEnumerator0
    )
    UsePositionOffset=false
    Write Material Random=true
   )
   &CameraOffset(Camera Offset Amount=&MergeHelper_FloatSwitch(0=-25, 1=-20, 2=?, 3=?, Index=&SpawnGroup(), Num=::NewEnumerator0))
   &ShapeLocation(
    Hemisphere Distribution=&MergeHelper_Vector2Switch(0=<1 0.8>, 1=<1 1>, 2=?, 3=?, 4=?, Index=&SpawnGroup(), Num=::NewEnumerator0)
    MaskSpawnGroup=false
    Override Seed=false
    Shape Origin=<0 0 0>
    Sphere Radius=&MergeHelper_FloatSwitch(
     0=&MakeCustomFloatFromBool(Bool=$User.IsLargeBody, True Float=20, False Float=18)
     1=&MakeCustomFloatFromBool(Bool=$User.IsLargeBody, True Float=5, False Float=4)
     2=?
     3=?
     Index=&SpawnGroup()
     Num=::NewEnumerator0
    )
    Sphere Surface Distribution=1
    UseOwnerScale=true
   )
   &AddVelocity(
    Module.Use Falloff Distance=false
    Origin Offset=<0 0 -10>
    Origin Offset Coordinate Space=::Simulation
    Override Seed=false
    Velocity Mode=::NewEnumerator1
    Velocity Origin=<0 0 0>
    Velocity Speed=&MergeHelper_FloatSwitch(
     0=&RandomRangeFloat(Minimum=10, Maximum=20)
     1=0
     2=?
     3=?
     Index=&SpawnGroup()
     Num=::NewEnumerator0
    )
   ) {module=AddVelocity001}
  ]
  particleUpdateScript [
   &ParticleState()
   &ScaleSpriteSize(
    Scale Sprite Size Mode=::NewEnumerator0
    Uniform Scale Factor=&MergeHelper_FloatSwitch(
     0=&FloatFromCurve(FloatCurve=DataInterface() {propertyValues {Curve {
         defaultValue=1
         keys [
          {arriveTangent=1.5038, interpMode=RCIM_Cubic, leaveTangent=1.5038, tangentMode=RCTM_User, time=0, value=0.75}
          {arriveTangent=0.2735, interpMode=RCIM_Cubic, leaveTangent=0.2735, time=0.1963, value=0.8713}
          {arriveTangent=0.3468, interpMode=RCIM_Cubic, leaveTangent=0.3468, time=0.7984, value=0.9684}
          {arriveTangent=1.8818, interpMode=RCIM_Cubic, leaveTangent=1.8818, tangentMode=RCTM_User, time=1, value=1.15}
         ]
        }}})
     1=&FloatFromCurve(FloatCurve=DataInterface() {propertyValues {Curve {
         defaultValue=1
         keys [
          {arriveTangent=1.5038, interpMode=RCIM_Cubic, leaveTangent=1.5038, tangentMode=RCTM_User, time=0, value=0.75}
          {arriveTangent=0.2205, interpMode=RCIM_Cubic, leaveTangent=0.2205, time=0.1963, value=0.8713}
          {arriveTangent=0.1601, interpMode=RCIM_Cubic, leaveTangent=0.1601, time=0.7984, value=0.926}
          {arriveTangent=1.8818, interpMode=RCIM_Cubic, leaveTangent=1.8818, tangentMode=RCTM_User, time=1, value=1}
         ]
        }}})
     2=?
     3=?
     Index=&SpawnGroup()
     Num=::NewEnumerator0
    )
   )
   &SpriteRotationRate(Rotation Rate=&MergeHelper_FloatSwitch(
     0=30
     1=&Multiply_Float(
      A=&RandomRangeFloat(Minimum=10, Maximum=25)
      B=&SelectFloatFromArray(
       Array Sampling Mode=::NewEnumerator0
       Direct Array Index=?
       Float Selection Array=DataInterface() {propertyValues {FloatData [-1,1]}}
       Random Seed=0
      )
     )
     2=?
     3=?
     Index=&SpawnGroup()
     Num=::NewEnumerator0
    ))
   &ScaleColor(
    Scale Alpha=&MergeHelper_FloatSwitch(
     0=&Multiply_Float(
      A=&FloatFromCurve(
       FloatCurve=DataInterface() {propertyValues {Curve {keys [
           {interpMode=RCIM_Cubic, leaveTangentWeight=0.2338, tangentMode=RCTM_Break, time=0, value=0}
           <at 0.3 1 RCIM_Cubic>
           {arriveTangentWeight=0.3188, interpMode=RCIM_Cubic, tangentMode=RCTM_User, time=1, value=0}
          ]}}}
       Scale Curve=&MakeCustomFloatFromBool(Bool=$User.IsFrontEnd, True Float=0.7, False Float=1)
      )
      B=&FloatFromCurve(
       FloatCurve=DataInterface() {propertyValues {Curve {keys [
           <at 0.1 0.85 RCIM_Cubic>
           {arriveTangentWeight=0.1077, interpMode=RCIM_Cubic, leaveTangentWeight=0.0718, tangentMode=RCTM_User, time=0.4, value=1}
           {arriveTangentWeight=0.0718, interpMode=RCIM_Cubic, leaveTangentWeight=0.1077, tangentMode=RCTM_User, time=0.6, value=1}
           <at 0.9 0.85 RCIM_Cubic>
          ]}}}
       CurveIndex=Linked() {linkedVariable {name=NPC.FortniteNPC.TimeOfDay_0-1, type=NiagaraFloat}}
      )
     )
     1=&Multiply_Float(
      A=&FloatFromCurve(
       FloatCurve=DataInterface() {propertyValues {Curve {keys [
           {interpMode=RCIM_Cubic, leaveTangent=1.1717, leaveTangentWeight=0.2338, tangentMode=RCTM_Break, time=0, value=0}
           <at 0.3 1 RCIM_Cubic>
           {arriveTangent=-0.8056, arriveTangentWeight=0.3188, interpMode=RCIM_Cubic, leaveTangent=-0.8056, tangentMode=RCTM_User, time=1, value=0}
          ]}}}
       Scale Curve=&MakeCustomFloatFromBool(Bool=$User.IsFrontEnd, True Float=1.1, False Float=1)
      )
      B=&FloatFromCurve(
       FloatCurve=DataInterface() {propertyValues {Curve {keys [
           <at 0.1 0.95 RCIM_Cubic>
           {arriveTangentWeight=0.1077, interpMode=RCIM_Cubic, leaveTangentWeight=0.0718, tangentMode=RCTM_User, time=0.4, value=1}
           {arriveTangentWeight=0.0718, interpMode=RCIM_Cubic, leaveTangentWeight=0.1077, tangentMode=RCTM_User, time=0.6, value=1}
           <at 0.9 0.95 RCIM_Cubic>
          ]}}}
       CurveIndex=Linked() {linkedVariable {name=NPC.FortniteNPC.TimeOfDay_0-1, type=NiagaraFloat}}
      )
     )
     2=?
     3=?
     Index=&SpawnGroup()
     Num=::NewEnumerator0
    )
    Scale RGB=&MergeHelper_VectorSwitch(
     0=&Multiply_Vector(
      A=&MakeVectorFromLinearColorRGB(Color=&ColorFromCurve(ColorCurve=DataInterface() {propertyValues {
          AlphaCurve {keys [<at 0 1 RCIM_Cubic>]}
          BlueCurve {keys [<at 0 1>,<at 0.331 1.5 RCIM_Cubic>,<at 1 1 RCIM_Cubic>]}
          GreenCurve {keys [<at 0 0.8441>, {arriveTangent=0.1417, interpMode=RCIM_Cubic, leaveTangent=0.1417, time=0.331, value=1.5}, <at 1 0.9857 RCIM_Cubic>]}
          RedCurve {keys [<at 0 0.5318>, {arriveTangent=-0.5318, interpMode=RCIM_Cubic, leaveTangent=-0.5318, time=0.331, value=1.5}, <at 1 0 RCIM_Cubic>]}
         }}))
      B=&MakeVectorFromLinearColorRGB(Color=&ColorFromCurve(
        ColorCurve=$System.ColorPalette3 Desaturated
        CurveIndex=&Frac_Float(Float=&Add_Float(A=$Particles.NormalizedAge, B=&RandomRangeFloat(Minimum=0, Maximum=1)))
       ))
     )
     1=&Multiply_Vector(
      A=&MakeVectorFromLinearColorRGB(Color=&ColorFromCurve(ColorCurve=DataInterface() {propertyValues {
          AlphaCurve {keys [<at 0 1 RCIM_Cubic>]}
          BlueCurve {keys [<at 0 1>,<at 0.331 1.7 RCIM_Cubic>,<at 1 1 RCIM_Cubic>]}
          GreenCurve {keys [<at 0 0.8441>, {arriveTangent=0.1417, interpMode=RCIM_Cubic, leaveTangent=0.1417, time=0.331, value=1.7}, <at 1 0.9857 RCIM_Cubic>]}
          RedCurve {keys [<at 0 0.5318>, {arriveTangent=-0.5318, interpMode=RCIM_Cubic, leaveTangent=-0.5318, time=0.331, value=1.7}, <at 1 0 RCIM_Cubic>]}
         }}))
      B=&MakeVectorFromLinearColorRGB(Color=&ColorFromCurve(
        ColorCurve=$System.ColorPalette3 Desaturated
        CurveIndex=&Frac_Float(Float=&Add_Float(A=$Particles.NormalizedAge, B=&RandomRangeFloat(Minimum=0, Maximum=1)))
       ))
     )
     2=?
     3=?
     Index=&SpawnGroup()
     Num=::NewEnumerator0
    )
    ScaleA=true
    ScaleRGB=true
   ) {module=ScaleColor001}
   &DynamicMaterialParameters(
    Index 0 Param 1=&RandomRangeFloat(Minimum=0, Maximum=3)
    Module.Index0Float0Write=true
    Module.Index0Float1Write=true
    Module.Index0Float2Write=true
    Module.Index0Float3Write=true
   )
   &Drag(Drag=0.5, Use Linear Drag=true, Use Rotational Drag=true)
   &SolveForcesAndVelocity(Clamp Velocity=false, Limit Acceleration=false)
   /SetVariables(Particles.PositionMod:NiagaraPosition=&AddVectorToPosition(
     Position=$Engine.Emitter.SimulationPosition
     Vector=&Add_Vector(
      A=&TransformVector(Vector=Linked() {linkedVariable {name=System.CenteredPos, type=NiagaraPosition}}, Destination Space=::Simulation)
      B=Linked() {linkedVariable {name=Particles.Position, type=NiagaraPosition}}
     )
    )) {isSetParameters=true, module=SetVariables_#21}
  ]
  rendererClasses [&NiagaraSpriteRendererProperties]
  renderers [
   {rendererClass=&NiagaraSpriteRendererProperties}
   {rendererClass=&NiagaraSpriteRendererProperties, rendererIndex=1}
   {rendererClass=&NiagaraSpriteRendererProperties, rendererIndex=2}
  ]
 }
 Stars_Core {
  bEnabled=false
  emitterUpdateScript [
   &EmitterState()
   &SpawnBurst_Instantaneous(Spawn Count=10, Use Spawn Probability=false, UseLoopCountLimit=true)
   &SpawnRate(Spawn Probability=0.9, SpawnRate=35, Use Spawn Probability=true)
   /SetVariables(Emitter.TwinkleCurve=DataInterface() {propertyValues {Curve {keys [
        {arriveTangent=12.2941, interpMode=RCIM_Cubic, leaveTangent=12.2941, tangentMode=RCTM_User, time=-0.0087, value=0.2976}
        {arriveTangent=-0.295, interpMode=RCIM_Cubic, leaveTangent=-0.295, time=0.1189, value=1}
        <at 1 0 RCIM_Cubic>
       ]}}}) {isSetParameters=true, module=SetVariables_#22}
  ]
  particleSpawnScript [
   /SetVariables(Particles.PartitionPicker:NiagaraFloat=&FloatFromCurve(
     FloatCurve=DataInterface() {propertyValues {Curve {defaultValue=1, keys [{tangentMode=RCTM_User, time=0.3646, value=1}, {tangentMode=RCTM_User, time=1, value=0}]}}}
     CurveIndex=&RandomRangeFloat(Minimum=0, Maximum=1)
    )) {isSetParameters=true, module=SetVariables_#23}
   /SetVariables(Particles.PartitionPicker=1) {enabled=false, isSetParameters=true, module=SetVariables_#24}
   &PartitionParticles(A:NiagaraFloat=$Particles.PartitionPicker, B=0.5, Partition Mode:ENiagara_PartitionMode_v2=::NewEnumerator4)
   /SetVariables(Particles.VisibilityTag=&MergeHelper_IntSwitch(
     0=/Niagara/DynamicInputs/Arrays/SelectIntFromWeightedDistributionArray(
      Array Sampling Mode:ENiagara_LifetimeMode=::NewEnumerator1
      Direct Array Index:NiagaraInt32=?
      Direct Array Mode:ENiagara_WrapClamp=::NewEnumerator0
      Evaluation Type:ENiagaraRandomnessEvaluation=::NewEnumerator0
      Recalculate Random Each Loop:NiagaraBool=false
      Weighted Distribution Int Array:NiagaraDataInterfaceArrayDistributionInt=DataInterface() {propertyValues {ArrayData [{value=0, weight=0.12}, {value=1, weight=0.88}]}}
      Randomness Mode:ENiagaraRandomnessMode=::NewEnumerator0
      Random Seed:NiagaraInt32=0
      Fixed Random Seed:NiagaraBool=false
     )
     1=1
     2=?
     3=?
     4=?
     5=?
     Index=$Particles.PartitionParticles.Partition
     Num=::NewEnumerator0
    )) {isSetParameters=true, module=SetVariables_#25}
   &InitializeParticle(
    Color Channel Mode:ENiagaraChannelCorrelation=::NewEnumerator2
    Color Maximum=<rgba 7.352 8.9604 12 1>
    Color Minimum=<rgba 3 1.9096 2.0533 1>
    Color Mode=::NewEnumerator2
    Lifetime=&MergeHelper_FloatSwitch(
     0=&RandomRangeFloat(Minimum=0.8, Maximum=1.5)
     1=&RandomRangeFloat(Minimum=0.6, Maximum=1)
     2=?
     3=?
     Index=$Particles.PartitionParticles.Partition
     Num=::NewEnumerator0
    )
    Sprite Size Mode=::NewEnumerator3
    Uniform Sprite Size=&MergeHelper_FloatSwitch(
     0=&FloatFromCurve(
      FloatCurve=DataInterface() {propertyValues {Curve {keys [<at 0 10>,<at 0.8 17>,<at 1 30>]}}}
      CurveIndex=&RandomRangeFloat(Minimum=0, Maximum=1)
     )
     1=&RandomRangeFloat(Minimum=1, Maximum=6)
     2=?
     3=?
     Index=$Particles.VisibilityTag
     Num=::NewEnumerator0
    )
    UsePositionOffset=false
    Write Material Random=true
   )
   &ShapeLocation(
    Hemisphere Distribution=<1 0.7>
    MaskSpawnGroup=false
    Override Seed=false
    Shape Origin=&MergeHelper_VectorSwitch(0=<0 0 0>, 1=<0 0 -2>, 2=?, 3=?, Index=$Particles.PartitionParticles.Partition, Num=::NewEnumerator0)
    Sphere Radius=&MergeHelper_FloatSwitch(
     0=&MakeCustomFloatFromBool(Bool=$User.IsLargeBody, True Float=20, False Float=28)
     1=&MakeCustomFloatFromBool(Bool=$User.IsLargeBody, True Float=12, False Float=11)
     2=?
     3=?
     Index=$Particles.PartitionParticles.Partition
     Num=::NewEnumerator0
    )
    Sphere Surface Distribution=1
    UseOwnerScale=true
   )
   &AddVelocity(
    Module.Use Falloff Distance=false
    Origin Offset=<0 0 -20>
    Origin Offset Coordinate Space=::Simulation
    Override Seed=false
    Velocity Mode=::NewEnumerator1
    Velocity Origin=<0 0 0>
    Velocity Speed=&Multiply_Float(
     A=&MergeHelper_FloatSwitch(
      0=&RandomRangeFloat(Minimum=10, Maximum=20)
      1=&RandomRangeFloat(Minimum=35, Maximum=50)
      2=?
      3=?
      Index=$Particles.PartitionParticles.Partition
      Num=::NewEnumerator0
     )
     B=&MakeCustomFloatFromBool(Bool=$User.IsLargeBody, True Float=1.1, False Float=1)
    )
   ) {module=AddVelocity001}
   &DynamicMaterialParameters(
    Index 0 Param 1=0.2
    Index 0 Param 2=&RandomRangeFloat(Minimum=0.3, Maximum=0.8)
    Index 0 Param 3=&RandomRangeFloat(Minimum=0.25, Maximum=0.5)
    Index 0 Param 4=1
    Module.Index0Float0Write=true
    Module.Index0Float1Write=true
    Module.Index0Float2Write=true
    Module.Index0Float3Write=true
   )
   &CameraOffset(Camera Offset Amount=-39)
  ]
  particleUpdateScript [
   &ParticleState()
   &Drag(
    Drag=&MergeHelper_FloatSwitch(0=0.5, 1=1, 2=?, 3=?, Index=$Particles.PartitionParticles.Partition, Num=::NewEnumerator0)
    Use Linear Drag=true
    Use Rotational Drag=true
   ) {module=Drag001}
   &SolveForcesAndVelocity(Clamp Velocity=false, Limit Acceleration=false)
   &ScaleColor(
    Scale Alpha=&FloatFromCurve(FloatCurve=DataInterface() {propertyValues {Curve {
        defaultValue=1
        keys [
         {arriveTangent=6.1295, interpMode=RCIM_Cubic, leaveTangent=6.1295, tangentMode=RCTM_User, time=0, value=0}
         {interpMode=RCIM_Cubic, tangentMode=RCTM_User, time=0.2, value=1}
         {arriveTangentWeight=0.2906, interpMode=RCIM_Cubic, leaveTangentWeight=0.1163, tangentMode=RCTM_User, time=0.75, value=1}
         <at 1 0 RCIM_Cubic>
        ]
       }}})
    ScaleA=true
    ScaleRGB=true
   )
   &ScaleSpriteSize(Uniform Curve Sprite Scale=DataInterface() {propertyValues {Curve {keys [
        {arriveTangent=3.5058, interpMode=RCIM_Cubic, leaveTangent=3.5058, leaveTangentWeight=0.4008, tangentMode=RCTM_User, time=0, value=0.1}
        {
         arriveTangent=0.1694
         arriveTangentWeight=0.1237
         interpMode=RCIM_Cubic
         leaveTangent=0.1694
         leaveTangentWeight=0.1649
         tangentMode=RCTM_User
         time=0.3
         value=1
        }
        {arriveTangentWeight=0.158, interpMode=RCIM_Cubic, leaveTangent=1.2043, leaveTangentWeight=0.1185, tangentMode=RCTM_Break, time=0.7, value=1}
        {arriveTangent=-2.5625, arriveTangentWeight=0.1237, interpMode=RCIM_Cubic, leaveTangent=-2.5625, tangentMode=RCTM_User, time=1, value=0.3733}
       ]}}})
   /SetVariables(Particles.PositionMod:NiagaraPosition=&AddVectorToPosition(
     Position=$Engine.Emitter.SimulationPosition
     Vector=&Add_Vector(
      A=&TransformVector(Vector=Linked() {linkedVariable {name=System.CenteredPos, type=NiagaraPosition}}, Destination Space=::Simulation)
      B=Linked() {linkedVariable {name=Particles.Position, type=NiagaraPosition}}
     )
    )) {isSetParameters=true, module=SetVariables_#26}
  ]
  rendererClasses [&NiagaraSpriteRendererProperties]
  renderers [
   {rendererClass=&NiagaraSpriteRendererProperties}
   {rendererClass=&NiagaraSpriteRendererProperties, rendererIndex=1}
  ]
 }
 Stars_Core001 {
  bEnabled=false
  emitterUpdateScript [
   &EmitterState()
   &SpawnBurst_Instantaneous(Spawn Count=10, Use Spawn Probability=false, UseLoopCountLimit=true)
   &SpawnRate(Spawn Probability=0.9, SpawnRate=35, Use Spawn Probability=true)
   /SetVariables(Emitter.TwinkleCurve=DataInterface() {propertyValues {Curve {keys [
        {arriveTangent=12.2941, interpMode=RCIM_Cubic, leaveTangent=12.2941, tangentMode=RCTM_User, time=-0.0087, value=0.2976}
        {arriveTangent=-0.295, interpMode=RCIM_Cubic, leaveTangent=-0.295, time=0.1189, value=1}
        <at 1 0 RCIM_Cubic>
       ]}}}) {isSetParameters=true, module=SetVariables_#22}
  ]
  particleSpawnScript [
   /SetVariables(Particles.PartitionPicker:NiagaraFloat=&FloatFromCurve(
     FloatCurve=DataInterface() {propertyValues {Curve {defaultValue=1, keys [{tangentMode=RCTM_User, time=0.3646, value=1}, {tangentMode=RCTM_User, time=1, value=0}]}}}
     CurveIndex=&RandomRangeFloat(Minimum=0, Maximum=1)
    )) {isSetParameters=true, module=SetVariables_#23}
   /SetVariables(Particles.PartitionPicker=1) {enabled=false, isSetParameters=true, module=SetVariables_#24}
   &PartitionParticles(A:NiagaraFloat=$Particles.PartitionPicker, B=0.5, Partition Mode:ENiagara_PartitionMode_v2=::NewEnumerator4)
   /SetVariables(Particles.VisibilityTag=&MergeHelper_IntSwitch(
     0=&SelectIntFromWeightedDistributionArray(
      Array Sampling Mode:ENiagara_LifetimeMode=::NewEnumerator1
      Direct Array Index:NiagaraInt32=?
      Direct Array Mode:ENiagara_WrapClamp=::NewEnumerator0
      Evaluation Type:ENiagaraRandomnessEvaluation=::NewEnumerator0
      Recalculate Random Each Loop:NiagaraBool=false
      Weighted Distribution Int Array:NiagaraDataInterfaceArrayDistributionInt=DataInterface() {propertyValues {ArrayData [{value=0, weight=0.12}, {value=1, weight=0.88}]}}
      Randomness Mode:ENiagaraRandomnessMode=::NewEnumerator0
      Random Seed:NiagaraInt32=0
      Fixed Random Seed:NiagaraBool=false
     )
     1=1
     2=?
     3=?
     4=?
     5=?
     Index=$Particles.PartitionParticles.Partition
     Num=::NewEnumerator0
    )) {isSetParameters=true, module=SetVariables_#25}
   &InitializeParticle(
    Color Channel Mode:ENiagaraChannelCorrelation=::NewEnumerator2
    Color Maximum=<rgba 12 6.4508 6.7876 1>
    Color Minimum=<rgba 2.9188 3 0.6084 1>
    Color Mode=::NewEnumerator2
    Lifetime=&MergeHelper_FloatSwitch(
     0=&RandomRangeFloat(Minimum=0.8, Maximum=1.5)
     1=&RandomRangeFloat(Minimum=0.6, Maximum=1)
     2=?
     3=?
     Index=$Particles.PartitionParticles.Partition
     Num=::NewEnumerator0
    )
    Sprite Size Mode=::NewEnumerator3
    Uniform Sprite Size=&MergeHelper_FloatSwitch(
     0=&FloatFromCurve(
      FloatCurve=DataInterface() {propertyValues {Curve {keys [<at 0 10>,<at 0.8 17>,<at 1 30>]}}}
      CurveIndex=&RandomRangeFloat(Minimum=0, Maximum=1)
     )
     1=&RandomRangeFloat(Minimum=1, Maximum=6)
     2=?
     3=?
     Index=$Particles.VisibilityTag
     Num=::NewEnumerator0
    )
    UsePositionOffset=false
    Write Material Random=true
   )
   &ShapeLocation(
    Hemisphere Distribution=<1 0.7>
    MaskSpawnGroup=false
    Override Seed=false
    Shape Origin=&MergeHelper_VectorSwitch(0=<0 0 0>, 1=<0 0 -2>, 2=?, 3=?, Index=$Particles.PartitionParticles.Partition, Num=::NewEnumerator0)
    Sphere Radius=&MergeHelper_FloatSwitch(
     0=&MakeCustomFloatFromBool(Bool=$User.IsLargeBody, True Float=20, False Float=28)
     1=&MakeCustomFloatFromBool(Bool=$User.IsLargeBody, True Float=12, False Float=11)
     2=?
     3=?
     Index=$Particles.PartitionParticles.Partition
     Num=::NewEnumerator0
    )
    Sphere Surface Distribution=1
    UseOwnerScale=true
   )
   &AddVelocity(
    Module.Use Falloff Distance=false
    Origin Offset=<0 0 -20>
    Origin Offset Coordinate Space=::Simulation
    Override Seed=false
    Velocity Mode=::NewEnumerator1
    Velocity Origin=<0 0 0>
    Velocity Speed=&Multiply_Float(
     A=&MergeHelper_FloatSwitch(
      0=&RandomRangeFloat(Minimum=10, Maximum=20)
      1=&RandomRangeFloat(Minimum=35, Maximum=50)
      2=?
      3=?
      Index=$Particles.PartitionParticles.Partition
      Num=::NewEnumerator0
     )
     B=&MakeCustomFloatFromBool(Bool=$User.IsLargeBody, True Float=1.1, False Float=1)
    )
   ) {module=AddVelocity001}
   &DynamicMaterialParameters(
    Index 0 Param 1=0.2
    Index 0 Param 2=&RandomRangeFloat(Minimum=0.3, Maximum=0.8)
    Index 0 Param 3=&RandomRangeFloat(Minimum=0.25, Maximum=0.5)
    Index 0 Param 4=1
    Module.Index0Float0Write=true
    Module.Index0Float1Write=true
    Module.Index0Float2Write=true
    Module.Index0Float3Write=true
   )
   &CameraOffset(Camera Offset Amount=-39)
  ]
  particleUpdateScript [
   &ParticleState()
   &Drag(
    Drag=&MergeHelper_FloatSwitch(0=0.5, 1=1, 2=?, 3=?, Index=$Particles.PartitionParticles.Partition, Num=::NewEnumerator0)
    Use Linear Drag=true
    Use Rotational Drag=true
   ) {module=Drag001}
   &SolveForcesAndVelocity(Clamp Velocity=false, Limit Acceleration=false)
   &ScaleColor(
    Scale Alpha=&FloatFromCurve(FloatCurve=DataInterface() {propertyValues {Curve {
        defaultValue=1
        keys [
         {arriveTangent=6.1295, interpMode=RCIM_Cubic, leaveTangent=6.1295, tangentMode=RCTM_User, time=0, value=0}
         {interpMode=RCIM_Cubic, tangentMode=RCTM_User, time=0.2, value=1}
         {arriveTangentWeight=0.2906, interpMode=RCIM_Cubic, leaveTangentWeight=0.1163, tangentMode=RCTM_User, time=0.75, value=1}
         <at 1 0 RCIM_Cubic>
        ]
       }}})
    ScaleA=true
    ScaleRGB=true
   )
   &ScaleSpriteSize(Uniform Curve Sprite Scale=DataInterface() {propertyValues {Curve {keys [
        {arriveTangent=3.5058, interpMode=RCIM_Cubic, leaveTangent=3.5058, leaveTangentWeight=0.4008, tangentMode=RCTM_User, time=0, value=0.1}
        {
         arriveTangent=0.1694
         arriveTangentWeight=0.1237
         interpMode=RCIM_Cubic
         leaveTangent=0.1694
         leaveTangentWeight=0.1649
         tangentMode=RCTM_User
         time=0.3
         value=1
        }
        {arriveTangentWeight=0.158, interpMode=RCIM_Cubic, leaveTangent=1.2043, leaveTangentWeight=0.1185, tangentMode=RCTM_Break, time=0.7, value=1}
        {arriveTangent=-2.5625, arriveTangentWeight=0.1237, interpMode=RCIM_Cubic, leaveTangent=-2.5625, tangentMode=RCTM_User, time=1, value=0.3733}
       ]}}})
   /SetVariables(Particles.PositionMod:NiagaraPosition=&AddVectorToPosition(
     Position=$Engine.Emitter.SimulationPosition
     Vector=&Add_Vector(
      A=&TransformVector(Vector=Linked() {linkedVariable {name=System.CenteredPos, type=NiagaraPosition}}, Destination Space=::Simulation)
      B=Linked() {linkedVariable {name=Particles.Position, type=NiagaraPosition}}
     )
    )) {isSetParameters=true, module=SetVariables_#26}
  ]
  rendererClasses [&NiagaraSpriteRendererProperties]
  renderers [
   {rendererClass=&NiagaraSpriteRendererProperties}
   {rendererClass=&NiagaraSpriteRendererProperties, rendererIndex=1}
  ]
 }
 Stars_Core002 {
  emitterUpdateScript [
   &EmitterState()
   &SpawnBurst_Instantaneous(Spawn Count=10, Use Spawn Probability=false, UseLoopCountLimit=true)
   &SpawnRate(Spawn Probability=0.9, SpawnRate=35, Use Spawn Probability=true)
   /SetVariables(Emitter.TwinkleCurve=DataInterface() {propertyValues {Curve {keys [
        {arriveTangent=12.2941, interpMode=RCIM_Cubic, leaveTangent=12.2941, tangentMode=RCTM_User, time=-0.0087, value=0.2976}
        {arriveTangent=-0.295, interpMode=RCIM_Cubic, leaveTangent=-0.295, time=0.1189, value=1}
        <at 1 0 RCIM_Cubic>
       ]}}}) {isSetParameters=true, module=SetVariables_#22}
  ]
  particleSpawnScript [
   /SetVariables(Particles.PartitionPicker:NiagaraFloat=&FloatFromCurve(
     FloatCurve=DataInterface() {propertyValues {Curve {defaultValue=1, keys [{tangentMode=RCTM_User, time=0.3646, value=1}, {tangentMode=RCTM_User, time=1, value=0}]}}}
     CurveIndex=&RandomRangeFloat(Minimum=0, Maximum=1)
    )) {isSetParameters=true, module=SetVariables_#23}
   /SetVariables(Particles.PartitionPicker=1) {enabled=false, isSetParameters=true, module=SetVariables_#24}
   &PartitionParticles(A:NiagaraFloat=$Particles.PartitionPicker, B=0.5, Partition Mode:ENiagara_PartitionMode_v2=::NewEnumerator4)
   /SetVariables(Particles.VisibilityTag=&MergeHelper_IntSwitch(
     0=&SelectIntFromWeightedDistributionArray(
      Array Sampling Mode:ENiagara_LifetimeMode=::NewEnumerator1
      Direct Array Index:NiagaraInt32=?
      Direct Array Mode:ENiagara_WrapClamp=::NewEnumerator0
      Evaluation Type:ENiagaraRandomnessEvaluation=::NewEnumerator0
      Recalculate Random Each Loop:NiagaraBool=false
      Weighted Distribution Int Array:NiagaraDataInterfaceArrayDistributionInt=DataInterface() {propertyValues {ArrayData [{value=0, weight=0.12}, {value=1, weight=0.88}]}}
      Randomness Mode:ENiagaraRandomnessMode=::NewEnumerator0
      Random Seed:NiagaraInt32=0
      Fixed Random Seed:NiagaraBool=false
     )
     1=1
     2=?
     3=?
     4=?
     5=?
     Index=$Particles.PartitionParticles.Partition
     Num=::NewEnumerator0
    )) {isSetParameters=true, module=SetVariables_#25}
   &InitializeParticle(
    Color Channel Mode:ENiagaraChannelCorrelation=::NewEnumerator2
    Color Maximum=<rgba 7.352 8.9604 12 1>
    Color Minimum=<rgba 3 1.9096 2.0533 1>
    Color Mode=::NewEnumerator2
    Lifetime=&MergeHelper_FloatSwitch(
     0=&RandomRangeFloat(Minimum=0.8, Maximum=1.5)
     1=&RandomRangeFloat(Minimum=0.6, Maximum=1)
     2=?
     3=?
     Index=$Particles.PartitionParticles.Partition
     Num=::NewEnumerator0
    )
    Sprite Size Mode=::NewEnumerator3
    Uniform Sprite Size=&MergeHelper_FloatSwitch(
     0=&FloatFromCurve(
      FloatCurve=DataInterface() {propertyValues {Curve {keys [<at 0 10>,<at 0.8 17>,<at 1 30>]}}}
      CurveIndex=&RandomRangeFloat(Minimum=0, Maximum=1)
     )
     1=&RandomRangeFloat(Minimum=1, Maximum=6)
     2=?
     3=?
     Index=$Particles.VisibilityTag
     Num=::NewEnumerator0
    )
    UsePositionOffset=false
    Write Material Random=true
   )
   &ShapeLocation(
    Hemisphere Distribution=<1 0.7>
    MaskSpawnGroup=false
    Override Seed=false
    Shape Origin=&MergeHelper_VectorSwitch(0=<0 0 0>, 1=<0 0 -2>, 2=?, 3=?, Index=$Particles.PartitionParticles.Partition, Num=::NewEnumerator0)
    Sphere Radius=&MergeHelper_FloatSwitch(
     0=&MakeCustomFloatFromBool(Bool=$User.IsLargeBody, True Float=20, False Float=28)
     1=&MakeCustomFloatFromBool(Bool=$User.IsLargeBody, True Float=12, False Float=11)
     2=?
     3=?
     Index=$Particles.PartitionParticles.Partition
     Num=::NewEnumerator0
    )
    Sphere Surface Distribution=1
    UseOwnerScale=true
   )
   &AddVelocity(
    Module.Use Falloff Distance=false
    Origin Offset=<0 0 -20>
    Origin Offset Coordinate Space=::Simulation
    Override Seed=false
    Velocity Mode=::NewEnumerator1
    Velocity Origin=<0 0 0>
    Velocity Speed=&Multiply_Float(
     A=&MergeHelper_FloatSwitch(
      0=&RandomRangeFloat(Minimum=10, Maximum=20)
      1=&RandomRangeFloat(Minimum=35, Maximum=50)
      2=?
      3=?
      Index=$Particles.PartitionParticles.Partition
      Num=::NewEnumerator0
     )
     B=&MakeCustomFloatFromBool(Bool=$User.IsLargeBody, True Float=1.1, False Float=1)
    )
   ) {module=AddVelocity001}
   &DynamicMaterialParameters(
    Index 0 Param 1=0.2
    Index 0 Param 2=&RandomRangeFloat(Minimum=0.3, Maximum=0.8)
    Index 0 Param 3=&RandomRangeFloat(Minimum=0.25, Maximum=0.5)
    Index 0 Param 4=1
    Module.Index0Float0Write=true
    Module.Index0Float1Write=true
    Module.Index0Float2Write=true
    Module.Index0Float3Write=true
   )
   &CameraOffset(Camera Offset Amount=-39)
  ]
  particleUpdateScript [
   &ParticleState()
   &Drag(
    Drag=&MergeHelper_FloatSwitch(0=0.5, 1=1, 2=?, 3=?, Index=$Particles.PartitionParticles.Partition, Num=::NewEnumerator0)
    Use Linear Drag=true
    Use Rotational Drag=true
   ) {module=Drag001}
   &SolveForcesAndVelocity(Clamp Velocity=false, Limit Acceleration=false)
   &ScaleColor(
    Scale Alpha=&FloatFromCurve(FloatCurve=DataInterface() {propertyValues {Curve {
        defaultValue=1
        keys [
         {arriveTangent=6.1295, interpMode=RCIM_Cubic, leaveTangent=6.1295, tangentMode=RCTM_User, time=0, value=0}
         {interpMode=RCIM_Cubic, tangentMode=RCTM_User, time=0.2, value=1}
         {arriveTangentWeight=0.2906, interpMode=RCIM_Cubic, leaveTangentWeight=0.1163, tangentMode=RCTM_User, time=0.75, value=1}
         <at 1 0 RCIM_Cubic>
        ]
       }}})
    ScaleA=true
    ScaleRGB=true
   )
   &ScaleSpriteSize(Uniform Curve Sprite Scale=DataInterface() {propertyValues {Curve {keys [
        {arriveTangent=3.5058, interpMode=RCIM_Cubic, leaveTangent=3.5058, leaveTangentWeight=0.4008, tangentMode=RCTM_User, time=0, value=0.1}
        {
         arriveTangent=0.1694
         arriveTangentWeight=0.1237
         interpMode=RCIM_Cubic
         leaveTangent=0.1694
         leaveTangentWeight=0.1649
         tangentMode=RCTM_User
         time=0.3
         value=1
        }
        {arriveTangentWeight=0.158, interpMode=RCIM_Cubic, leaveTangent=1.2043, leaveTangentWeight=0.1185, tangentMode=RCTM_Break, time=0.7, value=1}
        {arriveTangent=-2.5625, arriveTangentWeight=0.1237, interpMode=RCIM_Cubic, leaveTangent=-2.5625, tangentMode=RCTM_User, time=1, value=0.3733}
       ]}}})
   /SetVariables(Particles.PositionMod:NiagaraPosition=&AddVectorToPosition(
     Position=$Engine.Emitter.SimulationPosition
     Vector=&Add_Vector(
      A=&TransformVector(Vector=Linked() {linkedVariable {name=System.CenteredPos, type=NiagaraPosition}}, Destination Space=::Simulation)
      B=Linked() {linkedVariable {name=Particles.Position, type=NiagaraPosition}}
     )
    )) {isSetParameters=true, module=SetVariables_#26}
  ]
  rendererClasses [&NiagaraSpriteRendererProperties]
  renderers [
   {rendererClass=&NiagaraSpriteRendererProperties}
   {rendererClass=&NiagaraSpriteRendererProperties, rendererIndex=1}
  ]
 }
}
summary {
 systemName=NS_Aura_CosmicStars
 emitters [
  {bEnabled=false, emitterName=Stars_Core, rendererClasses [&NiagaraSpriteRendererProperties]}
  {bEnabled=false, emitterName=Radial_Beams, rendererClasses [&NiagaraSpriteRendererProperties]}
  {bEnabled=false, emitterName=Decal_Glows_Combined, rendererClasses [&NiagaraSpriteRendererProperties]}
  {bEnabled=false, emitterName=SpaceDust_Clouds, rendererClasses [&NiagaraSpriteRendererProperties]}
  {bEnabled=false, emitterName=Stars_Core001, rendererClasses [&NiagaraSpriteRendererProperties]}
  {bEnabled=false, emitterName=Radial_Beams001, rendererClasses [&NiagaraSpriteRendererProperties]}
  {bEnabled=false, emitterName=Decal_Glows_Combined001, rendererClasses [&NiagaraSpriteRendererProperties]}
  {bEnabled=false, emitterName=SpaceDust_Clouds001, rendererClasses [&NiagaraSpriteRendererProperties]}
  {emitterName=Stars_Core002, rendererClasses [&NiagaraSpriteRendererProperties]}
  {emitterName=Radial_Beams002, rendererClasses [&NiagaraSpriteRendererProperties]}
  {emitterName=Decal_Glows_Combined002, rendererClasses [&NiagaraSpriteRendererProperties]}
  {emitterName=SpaceDust_Clouds002, rendererClasses [&NiagaraSpriteRendererProperties]}
 ]
 userVariables [
  User.bIsFrontEnd {type=NiagaraBool}
  User.bIsFrontEndPreview {type=NiagaraBool}
  User.IsFrontEnd {type=NiagaraBool}
  User.IsLargeBody {type=NiagaraBool}
 ]
}
