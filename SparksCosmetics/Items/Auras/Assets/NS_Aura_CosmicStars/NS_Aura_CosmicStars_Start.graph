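# Start effect for the CosmicStars aura cosmetic: a short sprite burst of glow, flare, radial beams, space dust and stars on a skeletal mesh.
# Glow and flare follow the mesh, while beams, dust and stars sample it at spawn, with outward-moving stars also spawning continuously.
# Spawn groups, partitions and visibility tags pack variants into shared emitters.
# Time of day dims some layers, IsLargeBody resizes others, and system palettes tint beams and dust.
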NiagaraSystem /SparksCosmetics/Items/Auras/Assets/NS_Aura_CosmicStars/NS_Aura_CosmicStars_Start
dependencies {
 usedDataInterfaces [
  &NiagaraDataInterfaceColorCurve
  &NiagaraDataInterfaceCurve
  /Script/Niagara.NiagaraDataInterfaceArrayDistributionInt
  &NiagaraDataInterfaceArrayFloat
 ]
 usedDynamicInputs []
 usedModules [
  &self:SystemSpawnScript.SystemScriptSource.SystemScriptGraph.NiagaraNodeAssignment_0.SetVariables_#1
  &self:SystemSpawnScript.SystemScriptSource.SystemScriptGraph.NiagaraNodeAssignment_1.SetVariables_#2
  &self:SystemSpawnScript.SystemScriptSource.SystemScriptGraph.NiagaraNodeAssignment_3.SetVariables_#3
  &self:SystemSpawnScript.SystemScriptSource.SystemScriptGraph.NiagaraNodeAssignment_2.SetVariables_#4
  &SystemState
 ]
}
emitters {
 Corona_Glow_Combined {
  emitterUpdateScript [
   &EmitterState(Life Cycle Mode=::NewEnumerator1, Loop Behavior=::NewEnumerator1, Loop Duration=1, UseLoopDelay=false)
   &SpawnBurst_Instantaneous(Spawn Count=1, Use Spawn Probability=false)
   &SpawnBurst_Instantaneous(Spawn Count=1, Spawn Group=1, Use Spawn Probability=false) {module=SpawnBurst_Instantaneous001}
  ]
  particleSpawnScript [
   /SetVariables(Particles.VisibilityTag=&SpawnGroup()) {isSetParameters=true}
   &InitializeParticle(
    Color=&MergeHelper_ColorSwitch(
     0=<rgba 0.8791 0.6092 1 0.5>
     1=<rgba 0.0511 0.0153 0.065 1.1>
     2=?
     3=?
     Index=&SpawnGroup()
     Num=::NewEnumerator0
    )
    Lifetime=&MergeHelper_FloatSwitch(0=0.25, 1=0.4, 2=?, 3=?, Index=&SpawnGroup(), Num=::NewEnumerator0)
    Sprite Rotation Mode=::NewEnumerator2
    Sprite Size Mode=::NewEnumerator3
    Uniform Sprite Size=&MergeHelper_FloatSwitch(0=400, 1=200, 2=?, 3=?, Index=&SpawnGroup(), Num=::NewEnumerator0)
    UsePositionOffset=false
    Write Material Random=true
   )
   &CameraOffset(Camera Offset Amount=&MergeHelper_FloatSwitch(0=-30, 1=-10, 2=?, 3=?, Index=&SpawnGroup(), Num=::NewEnumerator0))
   &SystemLocation(
    MaskSpawnGroup=false
    Offset=&MergeHelper_VectorSwitch(0=<0 0 120>, 1=<0 0 0>, 2=?, 3=?, Index=&SpawnGroup(), Num=::NewEnumerator0)
   )
  ]
  particleUpdateScript [
   &ParticleState()
   &SkeletalMeshLocation(Bone Sampling Mode=::NewEnumerator4, MaskSpawnGroup=false, Skeletal Mesh=$System.Skeletal Mesh) {module=SkeletalMeshLocation001}
   &ScaleColor(
    Scale Alpha=&MergeHelper_FloatSwitch(
     0=&Multiply_Float(
      A=&FloatFromCurve(
       FloatCurve=DataInterface() {propertyValues {Curve {keys [
           <at 0 0.85 RCIM_Cubic>
           {arriveTangent=0.3, interpMode=RCIM_Cubic, leaveTangent=0.3, time=0.2009, value=0.92}
           <at 0.5 1 RCIM_Cubic>
           {arriveTangent=-0.3, interpMode=RCIM_Cubic, leaveTangent=-0.3, time=0.8028, value=0.92}
           <at 1 0.85 RCIM_Cubic>
          ]}}}
       CurveIndex=Linked() {linkedVariable {name=NPC.FortniteNPC.TimeOfDay_0-1, type=NiagaraFloat}}
      )
      B=&FloatFromCurve(FloatCurve=DataInterface() {propertyValues {Curve {keys [
           {arriveTangent=9.6062, interpMode=RCIM_Cubic, leaveTangent=9.6062, tangentMode=RCTM_User, time=0, value=0}
           {interpMode=RCIM_Cubic, tangentMode=RCTM_User, time=0.2, value=1}
           {interpMode=RCIM_Cubic, tangentMode=RCTM_User, time=1, value=0}
          ]}}})
     )
     1=&FloatFromCurve(
      FloatCurve=DataInterface() {propertyValues {Curve {keys [
          {interpMode=RCIM_Cubic, tangentMode=RCTM_User, time=0, value=0}
          {interpMode=RCIM_Cubic, tangentMode=RCTM_User, time=0.2, value=1}
          {interpMode=RCIM_Cubic, tangentMode=RCTM_User, time=0.5, value=1}
          {tangentMode=RCTM_User, time=1, value=0}
         ]}}}
      Scale Curve=2
     )
     2=?
     3=?
     Index=&SpawnGroup()
     Num=::NewEnumerator0
    )
    Scale RGB=&MergeHelper_VectorSwitch(
     0=&Multiply_VectorByFloat(
      Vector=&MakeVectorFromLinearColorRGB(Color=&ColorFromCurve(ColorCurve=DataInterface() {propertyValues {
          AlphaCurve {keys [<at 0 1 RCIM_Cubic>]}
          BlueCurve {keys [<at 0 0.5971 RCIM_Cubic>,<at 1 1 RCIM_Cubic>]}
          GreenCurve {keys [<at 0 0.5312 RCIM_Cubic>,<at 1 0.5348 RCIM_Cubic>]}
          RedCurve {keys [<at 0 1 RCIM_Cubic>,<at 1 0.7531 RCIM_Cubic>]}
         }}))
      Float=&FloatFromCurve(FloatCurve=DataInterface() {propertyValues {Curve {keys [{arriveTangent=-2.9031, interpMode=RCIM_Cubic, leaveTangent=-2.9031, tangentMode=RCTM_User, time=0, value=2}, <at 1 1 RCIM_Cubic>]}}})
     )
     1=&MakeVectorFromLinearColorRGB(Color=&ColorFromCurve(ColorCurve=DataInterface() {propertyValues {
         AlphaCurve {keys [<at 0 1 RCIM_Cubic>]}
         BlueCurve {keys [<at 0.1905 1 RCIM_Cubic>,<at 1 0.4974 RCIM_Cubic>]}
         GreenCurve {keys [<at 0.1905 1 RCIM_Cubic>,<at 1 0 RCIM_Cubic>]}
         RedCurve {keys [<at 0.1905 1 RCIM_Cubic>,<at 1 0.3623 RCIM_Cubic>]}
        }}))
     2=?
     3=?
     Index=&SpawnGroup()
     Num=::NewEnumerator0
    )
    ScaleA=true
    ScaleRGB=true
   )
   &ScaleSpriteSize(Scale Factor=&MergeHelper_Vector2Switch(
     0=&Vector2DFromFloat(Value=&FloatFromCurve(FloatCurve=DataInterface() {propertyValues {Curve {keys [
           {arriveTangent=2.1722, interpMode=RCIM_Cubic, leaveTangent=2.1722, tangentMode=RCTM_User, time=0, value=0}
           {arriveTangent=0.0224, interpMode=RCIM_Cubic, leaveTangent=0.0224, tangentMode=RCTM_User, time=1, value=1}
          ]}}}))
     1=<1 1>
     2=?
     3=?
     4=?
     Index=&SpawnGroup()
     Num=::NewEnumerator0
    ))
   &DynamicMaterialParameters(
    Index 0 Param 1=0.75
    Index 0 Param 2=2
    Index 0 Param 3=1
    Module.Index0Float0Write=true
    Module.Index0Float1Write=true
    Module.Index0Float2Write=true
    Module.Index0Float3Write=true
   ) {module=DynamicMaterialParameters001}
   /SetVariables(Particles.GlowDynamicParameter:Vector4f=<3 1 1 1>) {isSetParameters=true}
  ]
  rendererClasses [&NiagaraSpriteRendererProperties]
  renderers [
   {rendererClass=&NiagaraSpriteRendererProperties}
   {rendererClass=&NiagaraSpriteRendererProperties, rendererIndex=1}
  ]
 }
 Flare {
  emitterUpdateScript [
   &EmitterState(Life Cycle Mode=::NewEnumerator1, Loop Behavior=::NewEnumerator1, Loop Duration=1, UseLoopDelay=false)
   &SpawnBurst_Instantaneous(Spawn Count=1, Use Spawn Probability=false)
  ]
  particleSpawnScript [
   &InitializeParticle(
    Color=<rgba 1 0.5354 0.6584 1>
    Lifetime=0.3
    Sprite Rotation Mode=::NewEnumerator2
    Sprite Size Mode=::NewEnumerator3
    Uniform Sprite Size=250
    UsePositionOffset=false
    Write Material Random=true
   )
   &DynamicMaterialParameters(
    Index 0 Param 1=&RandomRangeFloat(Minimum=0.4, Maximum=0.6)
    Index 0 Param 2=1
    Index 0 Param 3=1
    Index 0 Param 4=1
    Module.Param0WriteEnabled=true
    Module.Param1WriteEnabled=true
    Module.Param2WriteEnabled=true
    Module.Param3WriteEnabled=true
   )
   &SystemLocation(MaskSpawnGroup=false, Offset=<0 0 120>)
   /SetVariables(Particles.Sharp Flare Camera Offset=40) {isSetParameters=true}
   &CameraOffset(Camera Offset Amount=-30) {module=CameraOffset001}
  ]
  particleUpdateScript [
   &ParticleState()
   &SolveForcesAndVelocity(Clamp Velocity=false, Limit Acceleration=false)
   &SkeletalMeshLocation(Bone Sampling Mode=::NewEnumerator4, MaskSpawnGroup=false, Skeletal Mesh=$System.Skeletal Mesh) {module=SkeletalMeshLocation001}
   &ScaleColor(
    Scale Alpha=&Multiply_Float(
     A=&FNTimeOfDayScalar(Night Scale=0.9, Morning Scale=0.95, Evening Scale=0.9)
     B=&FloatFromCurve(FloatCurve=DataInterface() {propertyValues {Curve {keys [<at 0 0>,<at 0.2 1>,<at 0.8 1>,<at 1 0>]}}})
    )
    Scale RGB=&Multiply_VectorByFloat(
     Vector=&MakeVectorFromLinearColorRGB(Color=&ColorFromCurve(ColorCurve=DataInterface() {propertyValues {
         AlphaCurve {keys [<at 0 1 RCIM_Cubic>]}
         BlueCurve {keys [<at 0 1.0039 RCIM_Cubic>,<at 1 0.5797 RCIM_Cubic>]}
         GreenCurve {keys [<at 0 1.7046 RCIM_Cubic>,<at 1 0.2687 RCIM_Cubic>]}
         RedCurve {keys [<at 0 2 RCIM_Cubic>,<at 1 1 RCIM_Cubic>]}
        }}))
     Float=1
    )
    ScaleA=true
    ScaleRGB=true
   )
   &ScaleSpriteSize(Scale Factor=&Vector2DFromFloat(Value=&FloatFromCurve(FloatCurve=DataInterface() {propertyValues {Curve {keys [<at 0 1.5 RCIM_Cubic>,<at 1 0>]}}})))
   /SetVariables(Particles.Flash color:LinearColor=&MakeLinearColorFromVectorAndFloat(
     "Vector (RGB)"=&MakeVectorFromLinearColorRGB(Color=<rgba 0.5235 0.3277 1 1>)
     "Float (Alpha)"=&Multiply_Float(
      A=&Multiply_Float(
       A=&FNTimeOfDayScalar(Night Scale=1, Morning Scale=1, Evening Scale=1)
       B=&FloatFromCurve(FloatCurve=DataInterface() {propertyValues {Curve {keys [<at 0 0>,<at 0.124 1>,<at 0.496 1>,<at 0.62 0>]}}})
      )
      B=1
     )
    )) {isSetParameters=true}
   /SetVariables(Particles.Flash size:Vector2f=&Multiply_Vector2D_ByFloat(
     Float=&Multiply_Float(
      A=400
      B=&FloatFromCurve(FloatCurve=DataInterface() {propertyValues {Curve {keys [
           {arriveTangent=4.1785, interpMode=RCIM_Cubic, leaveTangent=4.1785, tangentMode=RCTM_User, time=0, value=0}
           {arriveTangent=-0.15, interpMode=RCIM_Cubic, leaveTangent=-0.15, tangentMode=RCTM_User, time=0.1378, value=1}
           {arriveTangent=-0.0405, interpMode=RCIM_Cubic, leaveTangent=-0.0405, tangentMode=RCTM_User, time=0.62, value=0.25}
          ]}}})
     )
     Vector2D=<1 1>
    )) {isSetParameters=true}
  ]
  rendererClasses [&NiagaraSpriteRendererProperties]
  renderers [
   {rendererClass=&NiagaraSpriteRendererProperties}
   {rendererClass=&NiagaraSpriteRendererProperties, rendererIndex=1}
  ]
 }
 Radial_Beams {
  emitterUpdateScript [
   &EmitterState()
   &SpawnBurst_Instantaneous(Spawn Count=14, Use Spawn Probability=false, UseLoopCountLimit=true)
  ]
  particleSpawnScript [
   &PartitionParticles(Particles Per Partition=7, Partition Mode:ENiagara_PartitionMode_v2=::NewEnumerator1) {module=PartitionParticles001}
   &PartitionParticles(Number Of Partitions=3, Particles Per Partition=3, Partition Mode:ENiagara_PartitionMode_v2=::NewEnumerator1)
   /SetVariables(Particles.VisibilityTag=$Particles.PartitionParticles001.Partition) {isSetParameters=true}
   &InitializeParticle(
    Color=<rgba 1 1 1 0.8>
    Lifetime=&RandomRangeFloat(Minimum=0.22, Maximum=0.32)
    Sprite Rotation Angle=&Frac_Float(Float=&Add_Float(
      A=$Particles.PartitionParticles001.Normalized Index In Partition
      B=&MergeHelper_FloatSwitch(0=0, 1=0.2, 2=?, 3=?, Index=$Particles.PartitionParticles001.Partition, Num=::NewEnumerator0)
     ))
    Sprite Rotation Mode=::NewEnumerator1
    Sprite Size=&Multiply_Vector2D_ByFloat(
     Float=&MakeCustomFloatFromBool(Bool=$User.IsLargeBody, True Float=1.44, False Float=1.2)
     Vector2D=&MergeHelper_Vector2Switch(0=<40 65>, 1=<50 53>, 2=?, 3=?, 4=?, Index=$Particles.PartitionParticles001.Partition, Num=::NewEnumerator0)
    )
    Sprite Size Mode=::NewEnumerator4
    UsePositionOffset=false
    Write Material Random=true
   )
   &CameraOffset(Camera Offset Amount=-25)
   &SkeletalMeshLocation(Bone Sampling Mode=::NewEnumerator4, MaskSpawnGroup=false, Skeletal Mesh=$System.Skeletal Mesh)
  ]
  particleUpdateScript [
   &ParticleState()
   /SetVariables(Particles.UVIndex:NiagaraFloat=&SelectFloatFromArray(
     Array Sampling Mode=::NewEnumerator0
     Direct Array Index=?
     Float Selection Array=DataInterface() {propertyValues {FloatData [0,1,2,3]}}
     Random Seed=0
    )) {isSetParameters=true}
   /SetVariables(Particles.BeamScale:NiagaraFloat=&Waveform(
     Global Input Position=$Particles.PartitionParticles.Normalized Index In Partition
     "Waveform [1]"=::NewEnumerator0
     "[1] Phase Scale"=0.2
     "[1] Frequency"=1
     "[1] Amplitude Min/Max"=<0 1>
    )) {isSetParameters=true}
   &ScaleSpriteSize(
    Scale Factor=&Multiply_Vector2D(
     A=&MakeVector2D(
      X=/Game/Effects/Niagara/DynamicInputs/Math/FN_Fit(
       Clamp:NiagaraBool=false
       Value:NiagaraFloat=&Abs_Float(Float:NiagaraFloat=&OneMinusFloat(Float=$Particles.BeamScale))
       SourceMin:NiagaraFloat=0
       SourceMax:NiagaraFloat=1
       DestMin:NiagaraFloat=0.8
       DestMax:NiagaraFloat=1.6
      )
      Y=&FN_Fit(
       Clamp:NiagaraBool=false
       Value:NiagaraFloat=$Particles.BeamScale
       SourceMin:NiagaraFloat=0
       SourceMax:NiagaraFloat=1
       DestMin:NiagaraFloat=0.7
       DestMax:NiagaraFloat=1.3
      )
     )
     B=&MakeVector2D(
      X=&FloatFromCurve(FloatCurve=DataInterface() {propertyValues {Curve {
          defaultValue=1
          keys [
           {interpMode=RCIM_Cubic, tangentMode=RCTM_User, time=0, value=0.5}
           {interpMode=RCIM_Cubic, tangentMode=RCTM_User, time=0.5, value=1}
           {tangentMode=RCTM_User, time=1, value=1.3736}
          ]
         }}})
      Y=&FloatFromCurve(FloatCurve=DataInterface() {propertyValues {Curve {
          defaultValue=1
          keys [
           {interpMode=RCIM_Cubic, tangentMode=RCTM_User, time=0, value=0.5}
           {interpMode=RCIM_Cubic, tangentMode=RCTM_User, time=0.5, value=1}
           {tangentMode=RCTM_User, time=1.0141, value=1.1}
          ]
         }}})
     )
    )
    Scale Sprite Size Mode=::NewEnumerator2
   ) {module=ScaleSpriteSize001}
   &SolveForcesAndVelocity(Clamp Velocity=false, Limit Acceleration=false)
   &ScaleColor(
    Scale Alpha=&Multiply_Float(
     A=&Waveform(
      Global Input Position=$Particles.PartitionParticles.Normalized Index In Partition
      "Waveform [1]"=::NewEnumerator10
      "[1] Phase Scale"=0.4
      "[1] Frequency"=1
      "[1] Amplitude Min/Max"=<0.7 1>
     )
     B=&FloatFromCurve(FloatCurve=DataInterface() {propertyValues {Curve {
         defaultValue=1
         keys [{interpMode=RCIM_Cubic, tangentMode=RCTM_User, time=0, value=1}, {interpMode=RCIM_Cubic, tangentMode=RCTM_User, time=1, value=0}]
        }}})
    )
    Scale RGB=&MakeVectorFromLinearColorRGB(Color=&ColorFromCurve(
      ColorCurve=$System.ColorPalette3
      CurveIndex=&Frac_Float(Float=&Add_Float(A=&Multiply_Float(A=$Engine.Time, B=0.1), B=$Particles.PartitionParticles.Normalized Index In Partition))
     ))
    ScaleA=true
    ScaleRGB=true
   )
  ]
  rendererClasses [&NiagaraSpriteRendererProperties]
  renderers [
   {rendererClass=&NiagaraSpriteRendererProperties}
   {rendererClass=&NiagaraSpriteRendererProperties, rendererIndex=1}
  ]
 }
 Radial_SpaceDust {
  emitterUpdateScript [
   &EmitterState()
   &SpawnBurst_Instantaneous(Spawn Count=12, Use Spawn Probability=false, UseLoopCountLimit=true)
  ]
  particleSpawnScript [
   &PartitionParticles(Number Of Partitions=3, Particles Per Partition=4, Partition Mode:ENiagara_PartitionMode_v2=::NewEnumerator1)
   /SetVariables(Particles.VisibilityTag=&RandomRangeInt(Minimum=0, Maximum=1)) {isSetParameters=true}
   &InitializeParticle(
    Color=&MergeHelper_ColorSwitch(0=<rgba 1 1 1 0.35>, 1=<rgba 1.1 1.1 1.1 0.7>, 2=?, 3=?, Index=$Particles.VisibilityTag, Num=::NewEnumerator0)
    Lifetime Max=0.9
    Lifetime Min=0.4
    Lifetime Mode=::NewEnumerator1
    Sprite Rotation Mode=::NewEnumerator2
    Sprite Size Mode=::NewEnumerator3
    Sprite UV Mode=::NewEnumerator3
    Uniform Sprite Size=&Multiply_Float(
     A=&RandomRangeFloat(Minimum=45, Maximum=95)
     B=&MakeCustomFloatFromBool(Bool=$User.IsLargeBody, True Float=1.2, False Float=1)
    )
    UsePositionOffset=false
    Write Material Random=true
   )
   &CameraOffset(Camera Offset Amount=-35)
   &ShapeLocation(
    MaskSpawnGroup=false
    Shape Origin=<0 0 0>
    Sphere Distribution=::NewEnumerator3
    Sphere Radius=&MakeCustomFloatFromBool(Bool=$User.IsLargeBody, True Float=18, False Float=15)
    UseOwnerScale=true
   )
   &AddVelocity(
    Module.Use Falloff Distance=false
    Origin Offset=<0 0 -2>
    Origin Offset Coordinate Space=::Simulation
    Override Seed=false
    Velocity Mode=::NewEnumerator1
    Velocity Origin=<0 0 0>
    Velocity Speed=&RandomRangeFloat(Minimum=20, Maximum=45)
   ) {module=AddVelocity001}
   &SkeletalMeshLocation(
    Bone Sampling Mode=::NewEnumerator4
    MaskSpawnGroup=false
    Position Sampling=::NewEnumerator1
    Sampled Position Offset=&AddVectorToPosition(Position=$Particles.ShapeLocation.ShapePosition, Vector=<0 0 0>)
    Skeletal Mesh=$System.Skeletal Mesh
   )
  ]
  particleUpdateScript [
   &ParticleState()
   &ScaleSpriteSize(Uniform Curve Sprite Scale=DataInterface() {propertyValues {Curve {
       defaultValue=1
       keys [
        {arriveTangent=1.1384, interpMode=RCIM_Cubic, leaveTangent=1.1384, tangentMode=RCTM_User, time=0, value=0.75}
        {arriveTangent=0.4, interpMode=RCIM_Cubic, leaveTangent=0.4, time=0.7984, value=0.9684}
        {arriveTangent=1.8818, interpMode=RCIM_Cubic, leaveTangent=1.8818, tangentMode=RCTM_User, time=1, value=1.15}
       ]
      }}})
   &SpriteRotationRate(Rotation Rate=30)
   &ScaleColor(
    Scale Alpha=&Multiply_Float(
     A=&FloatFromCurve(
      FloatCurve=DataInterface() {propertyValues {Curve {keys [
          {interpMode=RCIM_Cubic, leaveTangentWeight=0.2338, tangentMode=RCTM_Break, time=0, value=0}
          <at 0.3 1 RCIM_Cubic>
          {arriveTangentWeight=0.3188, interpMode=RCIM_Cubic, tangentMode=RCTM_User, time=1, value=0}
         ]}}}
      Scale Curve=&MakeCustomFloatFromBool(Bool=$User.bIsFrontEndPreview, True Float=0.7, False Float=1)
     )
     B=&FloatFromCurve(
      FloatCurve=DataInterface() {propertyValues {Curve {keys [
          <at 0.1 0.85 RCIM_Cubic>
          {arriveTangentWeight=0.1077, interpMode=RCIM_Cubic, leaveTangentWeight=0.0718, tangentMode=RCTM_User, time=0.4, value=1}
          {arriveTangentWeight=0.0718, interpMode=RCIM_Cubic, leaveTangentWeight=0.1077, tangentMode=RCTM_User, time=0.6, value=1}
          <at 0.9 0.85 RCIM_Cubic>
         ]}}}
      CurveIndex=Linked() {linkedVariable {name=NPC.FortniteNPC.TimeOfDay_0-1, type=NiagaraFloat}}
     )
    )
    Scale RGB=&Multiply_Vector(
     A=&MakeVectorFromLinearColorRGB(Color=&ColorFromCurve(ColorCurve=DataInterface() {propertyValues {
         AlphaCurve {keys [<at 0 1 RCIM_Cubic>]}
         BlueCurve {keys [<at 0 1>,<at 0.331 1.5 RCIM_Cubic>,<at 1 1 RCIM_Cubic>]}
         GreenCurve {keys [<at 0 0.8441>, {arriveTangent=0.1417, interpMode=RCIM_Cubic, leaveTangent=0.1417, time=0.331, value=1.5}, <at 1 0.9857 RCIM_Cubic>]}
         RedCurve {keys [<at 0 0.5318>, {arriveTangent=-0.5318, interpMode=RCIM_Cubic, leaveTangent=-0.5318, time=0.331, value=1.5}, <at 1 0 RCIM_Cubic>]}
        }}))
     B=&MakeVectorFromLinearColorRGB(Color=&ColorFromCurve(
       ColorCurve=$System.ColorPalette3 Desaturated
       CurveIndex=&Frac_Float(Float=&Add_Float(A=$Particles.NormalizedAge, B=&RandomRangeFloat(Minimum=0, Maximum=1)))
      ))
    )
    ScaleA=true
    ScaleRGB=true
   ) {module=ScaleColor001}
   &DynamicMaterialParameters(
    Index 0 Param 1=&RandomRangeFloat(Minimum=0, Maximum=3)
    Module.Index0Float0Write=true
    Module.Index0Float1Write=true
    Module.Index0Float2Write=true
    Module.Index0Float3Write=true
   )
   &Drag(Use Linear Drag=true, Use Rotational Drag=true)
   &SolveForcesAndVelocity(Clamp Velocity=false, Limit Acceleration=false)
   /SetVariables(Particles.PositionMod:NiagaraPosition=&AddVectorToPosition(
     Position=$Engine.Emitter.SimulationPosition
     Vector=&Add_Vector(
      A=&TransformVector(Vector=Linked() {linkedVariable {name=System.CenteredPos, type=NiagaraPosition}}, Destination Space=::Simulation)
      B=Linked() {linkedVariable {name=Particles.Position, type=NiagaraPosition}}
     )
    )) {isSetParameters=true}
  ]
  rendererClasses [&NiagaraSpriteRendererProperties]
  renderers [
   {rendererClass=&NiagaraSpriteRendererProperties}
   {rendererClass=&NiagaraSpriteRendererProperties, rendererIndex=1}
  ]
 }
 Stars_Core {
  emitterUpdateScript [
   &EmitterState()
   &SpawnBurst_Instantaneous(Spawn Count=15, Use Spawn Probability=false, UseLoopCountLimit=true)
   &SpawnRate(Spawn Probability=0.9, SpawnRate=175, Use Spawn Probability=false)
  ]
  particleSpawnScript [
   /SetVariables(Particles.PartitionPicker:NiagaraFloat=&FloatFromCurve(
     FloatCurve=DataInterface() {propertyValues {Curve {defaultValue=1, keys [{tangentMode=RCTM_User, time=0.3, value=1}, {tangentMode=RCTM_User, time=1, value=0}]}}}
     CurveIndex=&RandomRangeFloat(Minimum=0, Maximum=1)
    )) {isSetParameters=true}
   &PartitionParticles(A:NiagaraFloat=$Particles.PartitionPicker, B=0.5, Partition Mode:ENiagara_PartitionMode_v2=::NewEnumerator4)
   /SetVariables(Particles.VisibilityTag=&MergeHelper_IntSwitch(
     0=/Niagara/DynamicInputs/Arrays/SelectIntFromWeightedDistributionArray(
      Array Sampling Mode:ENiagara_LifetimeMode=::NewEnumerator1
      Direct Array Index:NiagaraInt32=?
      Direct Array Mode:ENiagara_WrapClamp=::NewEnumerator0
      Evaluation Type:ENiagaraRandomnessEvaluation=::NewEnumerator0
      Recalculate Random Each Loop:NiagaraBool=false
      Weighted Distribution Int Array:NiagaraDataInterfaceArrayDistributionInt=DataInterface() {propertyValues {ArrayData [{value=0, weight=0.12}, {value=1, weight=0.88}]}}
      Randomness Mode:ENiagaraRandomnessMode=::NewEnumerator0
      Random Seed:NiagaraInt32=0
      Fixed Random Seed:NiagaraBool=false
     )
     1=1
     2=?
     3=?
     4=?
     5=?
     Index=$Particles.PartitionParticles.Partition
     Num=::NewEnumerator0
    )) {isSetParameters=true}
   &InitializeParticle(
    Color Channel Mode:ENiagaraChannelCorrelation=::NewEnumerator2
    Color Maximum=<rgba 7.352 8.9604 12 1>
    Color Minimum=<rgba 3 1.9096 2.0533 1>
    Color Mode=::NewEnumerator2
    Lifetime=&MergeHelper_FloatSwitch(
     0=&RandomRangeFloat(Minimum=0.5, Maximum=0.75)
     1=&RandomRangeFloat(Minimum=0.4, Maximum=0.9)
     2=?
     3=?
     Index=$Particles.PartitionParticles.Partition
     Num=::NewEnumerator0
    )
    Sprite Size Mode=::NewEnumerator3
    Uniform Sprite Size=&MergeHelper_FloatSwitch(
     0=&FloatFromCurve(
      FloatCurve=DataInterface() {propertyValues {Curve {keys [<at 0 10>,<at 0.8 17>,<at 1 30>]}}}
      CurveIndex=&RandomRangeFloat(Minimum=0, Maximum=1)
     )
     1=&RandomRangeFloat(Minimum=1, Maximum=6)
     2=?
     3=?
     Index=$Particles.VisibilityTag
     Num=::NewEnumerator0
    )
    UsePositionOffset=false
    Write Material Random=true
   )
   &DynamicMaterialParameters(
    Index 0 Param 1=0.2
    Index 0 Param 2=&RandomRangeFloat(Minimum=0.3, Maximum=0.8)
    Index 0 Param 3=&RandomRangeFloat(Minimum=0.25, Maximum=0.5)
    Index 0 Param 4=1
    Module.Index0Float0Write=true
    Module.Index0Float1Write=true
    Module.Index0Float2Write=true
    Module.Index0Float3Write=true
   )
   &CameraOffset(Camera Offset Amount=-35)
   &ShapeLocation(
    MaskSpawnGroup=false
    Override Seed=false
    Sphere Radius=&MergeHelper_FloatSwitch(
     0=&MakeCustomFloatFromBool(Bool=$User.IsLargeBody, True Float=20, False Float=28)
     1=&MakeCustomFloatFromBool(Bool=$User.IsLargeBody, True Float=12, False Float=11)
     2=?
     3=?
     Index=$Particles.PartitionParticles.Partition
     Num=::NewEnumerator0
    )
    Sphere Surface Distribution=1
    UseOwnerScale=true
   )
   &AddVelocity(
    Module.Use Falloff Distance=false
    Override Seed=false
    Velocity Mode=::NewEnumerator1
    Velocity Speed=&Multiply_Float(
     A=&MergeHelper_FloatSwitch(
      0=&RandomRangeFloat(Minimum=75, Maximum=150)
      1=&RandomRangeFloat(Minimum=125, Maximum=250)
      2=?
      3=?
      Index=$Particles.PartitionParticles.Partition
      Num=::NewEnumerator0
     )
     B=&MakeCustomFloatFromBool(Bool=$User.IsLargeBody, True Float=1.1, False Float=1)
    )
   ) {module=AddVelocity001}
   &SkeletalMeshLocation(
    Bone Sampling Mode=::NewEnumerator4
    MaskSpawnGroup=false
    Position Sampling=::NewEnumerator1
    Sampled Position Offset=&AddVectorToPosition(Position=$Particles.ShapeLocation.ShapePosition, Vector=<0 0 0>)
    Skeletal Mesh=$System.Skeletal Mesh
   )
  ]
  particleUpdateScript [
   &ParticleState()
   &Drag(
    Drag=&MergeHelper_FloatSwitch(0=2, 1=5, 2=?, 3=?, Index=$Particles.PartitionParticles.Partition, Num=::NewEnumerator0)
    Use Linear Drag=true
    Use Rotational Drag=true
   ) {module=Drag001}
   &SolveForcesAndVelocity(Clamp Velocity=false, Limit Acceleration=false)
   &ScaleColor(
    Scale Alpha=&FloatFromCurve(FloatCurve=DataInterface() {propertyValues {Curve {
        defaultValue=1
        keys [
         {arriveTangent=6.1295, interpMode=RCIM_Cubic, leaveTangent=6.1295, tangentMode=RCTM_User, time=0, value=0}
         {interpMode=RCIM_Cubic, tangentMode=RCTM_User, time=0.2, value=1}
         {arriveTangentWeight=0.2906, interpMode=RCIM_Cubic, leaveTangentWeight=0.1163, tangentMode=RCTM_User, time=0.75, value=1}
         <at 1 0 RCIM_Cubic>
        ]
       }}})
    ScaleA=true
    ScaleRGB=true
   )
   &ScaleSpriteSize(Uniform Curve Sprite Scale=DataInterface() {propertyValues {Curve {keys [
        {arriveTangent=3.5058, interpMode=RCIM_Cubic, leaveTangent=3.5058, leaveTangentWeight=0.4008, tangentMode=RCTM_User, time=0, value=0.1}
        {
         arriveTangent=0.1694
         arriveTangentWeight=0.1237
         interpMode=RCIM_Cubic
         leaveTangent=0.1694
         leaveTangentWeight=0.1649
         tangentMode=RCTM_User
         time=0.3
         value=1
        }
        {arriveTangentWeight=0.158, interpMode=RCIM_Cubic, leaveTangent=1.2043, leaveTangentWeight=0.1185, tangentMode=RCTM_Break, time=0.7, value=1}
        {arriveTangent=-2.5625, arriveTangentWeight=0.1237, interpMode=RCIM_Cubic, leaveTangent=-2.5625, tangentMode=RCTM_User, time=1, value=0.3733}
       ]}}})
  ]
  rendererClasses [&NiagaraSpriteRendererProperties]
  renderers [
   {rendererClass=&NiagaraSpriteRendererProperties}
   {rendererClass=&NiagaraSpriteRendererProperties, rendererIndex=1}
  ]
 }
}
summary {
 systemName=NS_Aura_CosmicStars_Start
 emitters [
  {emitterName=Flare, rendererClasses [&NiagaraSpriteRendererProperties]}
  {emitterName=Corona_Glow_Combined, rendererClasses [&NiagaraSpriteRendererProperties]}
  {emitterName=Stars_Core, rendererClasses [&NiagaraSpriteRendererProperties]}
  {emitterName=Radial_Beams, rendererClasses [&NiagaraSpriteRendererProperties]}
  {emitterName=Radial_SpaceDust, rendererClasses [&NiagaraSpriteRendererProperties]}
 ]
 userVariables [
  User.bIsFrontEndPreview {type=NiagaraBool}
  User.IsLargeBody {type=NiagaraBool}
  User.SkeletalMesh {
   type=NiagaraDataInterfaceSkeletalMesh
   defaultValue {
    struct=NiagaraExt_VariableValue_DataInterface
    value {dataInterface=&self:NiagaraDataInterfaceSkeletalMesh_1}
   }
  }
 ]
}
